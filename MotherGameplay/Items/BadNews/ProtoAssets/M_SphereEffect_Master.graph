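# Translucent master material for Bad News prototype sphere effects on Niagara mesh particles, with a Fresnel color gradient and glowing intersection edges.
# Two noise masks, offset, tiled and shaped by particle dynamic parameters (only Tile U is used), sum and combine additively or multiplicatively with Fresnel.
# That result, particle alpha and vertex red drive depth-faded opacity.
# Scene-depth edges, weighted toward sideways-facing normals, tint the particle-colored emissive.
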Material /MotherGameplay/Items/BadNews/ProtoAssets/M_SphereEffect_Master
graph {
 properties {bUsedWithNiagaraMeshParticles=true, blendMode=BLEND_Translucent, pixelDepthOffsetMode=PDOM_Legacy, usedShadingModels=Default Lit}
 nodes [
  n1 = ParticleColor()
  n2 = DynamicParameter() {paramNames [Energy Mask 01 Offset U, Energy Mask 01 Offset V, Energy Mask 01 Tile U, Energy Mask 01 Tile V]}
  n3 = AppendVector(A=ComponentMask(None=@n2.Energy Mask 01 Offset U) {r=true}, B=ComponentMask(None=@n2.Energy Mask 01 Offset V) {g=true})
  n4 = TextureCoordinate()
  n5 = TextureSampleParameter2D(UVs=Multiply(A=Add(A=@n4, B=@n3), B=AppendVector(A=@n2.Energy Mask 01 Tile U, B=@n2.Energy Mask 01 Tile U))) {parameterName=Energy Mask 01, texture=/Game/Global/Textures/Default/Caustics/T_Turbulence_Seamless}
  n6 = DynamicParameter() {
   parameterIndex=1
   defaultValue=<rgba 2 1 2 1>
   paramNames [Energy Mask 01 Power, Energy Mask 01 Strength, Energy Mask 02 Power, Energy Mask 02 Strength]
  }
  n7 = DynamicParameter() {parameterIndex=2, paramNames [Energy Mask 02 Offset U, Energy Mask 02 Offset V, Energy Mask 02 Tile U, Energy Mask 02 Tile V]}
  n8 = AppendVector(A=ComponentMask(None=@n7.Energy Mask 02 Offset U) {r=true}, B=ComponentMask(None=@n7.Energy Mask 02 Offset V) {g=true})
  n9 = TextureCoordinate()
  n10 = TextureSampleParameter2D(UVs=Multiply(A=Add(A=@n9, B=@n8), B=AppendVector(A=@n7.Energy Mask 02 Tile U, B=@n7.Energy Mask 02 Tile U))) {parameterName=Energy Mask 02, samplerType=SAMPLERTYPE_Masks, texture=/Game/Weapons/WeaponSkins/Wraps/Textures/Mash/Mash_Noise}
  n11 = Add(
   A=Multiply(A=Power(Base=@n5.G, Exp=@n6.Energy Mask 01 Power), B=@n6.Energy Mask 01 Strength)
   B=Multiply(A=Power(Base=@n10.G, Exp=@n6.Energy Mask 02 Power), B=@n6.Energy Mask 02 Strength)
  )
  n12 = NamedRerouteUsage()
  n13 = LinearInterpolate(A=Add(A=@n11, B=@n12), B=Multiply(A=@n11, B=@n12), Alpha=ScalarParameter() {parameterName=Multiply on Fresnel?})
  DepthFade(Opacity=Multiply(A=VertexColor().R, B=Multiply(A=@n1.A, B=@n13)), FadeDistance=ScalarParameter() {defaultValue=10, parameterName=Depth Fade})
  n14 = LinearInterpolate(
   A=VectorParameter() {parameterName=Fresnel Inner, defaultValue=<rgba 0.7434 0 1 1>}.RGB
   B=VectorParameter() {parameterName=Fresnel Outer, defaultValue=<rgba 0 0.9633 5 1>}.RGB
   Alpha=NamedRerouteUsage()
  )
  n15 = VectorParameter() {parameterName=Colored Edges, defaultValue=<rgba 0 15.1666 25 1>}
  n16 = NamedRerouteUsage()
  MaterialFunctionCall(Emissive=Multiply(A=LinearInterpolate(A=@n14, B=@n15.RGB, Alpha=@n16), B=ParticleColor().RGB)) {materialFunction=&MF_FX_EmissiveAdjustments}
  NamedRerouteDeclaration(None=Fresnel(
    ExponentIn=ScalarParameter() {defaultValue=5, parameterName=Fresnel Exp}
    BaseReflectFractionIn=ScalarParameter() {parameterName=Fresnel Base Reflect}
   ) {baseReflectFraction=0}) {name=Fresnel, nodeColor=<rgba 0 1 0.0625 1>}
  n17 = MaterialFunctionCall!(Distance Bias, Distance Divisor=ScalarParameter() {defaultValue=1, parameterName=IntersectEdgeDivisor}, Scene Depth, Pixel Depth) {materialFunction=/Game/Effects/Fort_Effects/Materials/_Functions/MF_DistanceBehindTranslucency_01}
  n18 = MaterialFunctionCall(Float3=VertexNormalWS()) {materialFunction=&BreakOutFloat3Components}
  n19 = MaterialFunctionCall(X=Abs(None=@n18.R), Y=Abs(None=@n18.G), Z=Saturate(None=@n18.B)) {materialFunction=&MakeFloat3}
  n20 = Constant3Vector() {constant=<rgba 1 1 0 0>}
  NamedRerouteDeclaration(None=Multiply(A=Power(Base=@n17.Normalized Inverted) {constExponent=3}, B=DotProduct(A=@n19.Result, B=@n20))) {name=IntersectionEdges, nodeColor=<rgba 1 0.1172 0 1>}
 ]
}
