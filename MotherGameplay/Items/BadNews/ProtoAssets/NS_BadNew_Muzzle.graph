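# Prototype muzzle effect for the BadNews item, built as a Niagara burst of blue-violet plasma.
# BoneFlow ribbons arc along User.SkeletalMesh sockets, while curl-noise contrail ribbons and speed-stretched sparkles shoot outward from the muzzle.
# A radial plasma sprite brightens when User.Hitting is set, and a corona flash with a dynamic light spawns at User.MinChargeDuration.
# The corona and sparkles drive extra per-renderer size and color attributes for layered sprites.
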NiagaraSystem /MotherGameplay/Items/BadNews/ProtoAssets/NS_BadNew_Muzzle
dependencies {
 usedDataInterfaces [&NiagaraDataInterfaceCurve, &NiagaraDataInterfaceColorCurve]
 usedDynamicInputs []
 usedModules [&SystemState]
}
emitters {
 BoneFlow_Fixed {
  emitterUpdateScript [
   &EmitterState()
   &SpawnRate(SpawnRate=20, Use Spawn Probability=false) {enabled=false}
   &SpawnBurst_Instantaneous(
    Spawn Count=&Multiply_Int(A=&RandomRangeInt(Minimum=1, Maximum=3, Evaluation Type=::NewEnumerator1), B=24)
    Use Spawn Probability=false
   ) {enabled=false}
   &SpawnBurst_Instantaneous(Spawn Count=20, Use Spawn Probability=false) {module=SpawnBurst_Instantaneous001}
  ]
  particleSpawnScript [
   /Niagara/Modules/Spawn/Initialization/InitializeParticle(
    Module.Write Color=true
    Module.Write Lifetime=true
    Module.Write Mass=true
    Module.Write Material Random=true
    Module.Write Position=false
    Module.Write Scale=false
    Module.Write SpriteRotation=false
    Module.Write SpriteSize=true
    Position=&SimulationPosition()
    Sprite Size=<5 5>
   )
   &InitializeRibbon(
    Color=$User.Color_Tint
    Lifetime=0.5
    Module.Write Color=true
    Module.Write Lifetime=true
    Module.Write Mass=true
    Module.Write Material Random=true
    Module.Write Position=false
    Module.Write RibbonTwist=false
    Module.Write RibbonWidth=true
    Ribbon Width=10
   )
   &FN_SetSpawnGroup(Inclusivity=true, Normalize Spawn Group=true, Number of Spawn Groups=2, Particles Per Group=10, Set Ribbon ID and Ribbon Link Order=true) {module=FN_SetSpawnGroup_v4}
   &AddVelocityFromPoint(Module.Use Falloff Distance=false, Origin Offset=<9 0 0>, Velocity Strength=500)
  ]
  particleUpdateScript [
   &ParticleState()
   &SolveForcesAndVelocity(Clamp Velocity=false, Limit Acceleration=false)
   &FN_BoneFlow(
    Angle:NiagaraFloat=&Add_Float(A=&Multiply_Float(A=$Particles.Normalized Spawn Group, B=1), B=$Particles.NormalizedAge)
    BoneID:NiagaraInt32=$Particles.SpawnGroup
    Height:NiagaraFloat=&FloatFromCurve(
     FloatCurve=DataInterface() {propertyValues {Curve {keys [<at 0 0 RCIM_Cubic>,<at 0.5 1 RCIM_Cubic>,<at 1 0 RCIM_Cubic>]}}}
     CurveIndex=$Particles.Normalized Spawn Group
     Scale Curve=0.5
    )
    Radius:NiagaraFloat=&FloatFromCurve(
     FloatCurve=DataInterface() {propertyValues {Curve {keys [
         {interpMode=RCIM_Cubic, tangentMode=RCTM_User, time=0, value=0}
         {interpMode=RCIM_Cubic, tangentMode=RCTM_User, time=0.5, value=1}
         {tangentMode=RCTM_User, time=1, value=0}
        ]}}}
     Scale Curve=46
    )
    SkeletalMesh:NiagaraDataInterfaceSkeletalMesh=$User.SkeletalMesh
    Use Sockets Instead=true
    Write Position Attribute=true
   )
   &ScaleRibbonWidth(Ribbon Width Scale=&Multiply_Float(
     A=&FloatFromCurve(
      FloatCurve=DataInterface() {propertyValues {Curve {keys [
          {arriveTangent=7.0965, interpMode=RCIM_Cubic, leaveTangent=7.0965, tangentMode=RCTM_User, time=0, value=0}
          {interpMode=RCIM_Cubic, tangentMode=RCTM_User, time=0.2, value=0.95}
          {tangentMode=RCTM_User, time=1, value=0}
         ]}}}
      Scale Curve=1.5
     )
     B=&FloatFromCurve(
      FloatCurve=DataInterface() {propertyValues {Curve {keys [
          {arriveTangent=-0.0707, interpMode=RCIM_Cubic, leaveTangent=-0.0707, tangentMode=RCTM_User, time=0, value=0}
          {interpMode=RCIM_Cubic, tangentMode=RCTM_User, time=0.5, value=0.9}
          {tangentMode=RCTM_User, time=1, value=0}
         ]}}}
      CurveIndex=$Particles.Normalized Spawn Group
     )
    ))
   &DynamicMaterialParameters(
    Index 0 Param 1=&FloatFromCurve(FloatCurve=DataInterface() {propertyValues {Curve {keys [
         <at -0.005 0.984 RCIM_Cubic>
         {arriveTangent=-0.9791, interpMode=RCIM_Cubic, leaveTangent=-0.9791, time=0.6792, value=0.8179}
         <at 1 0 RCIM_Cubic>
        ]}}})
    Index 0 Param 3=1
    Index 0 Param 4=$User.EmissiveIntensity
    Module.Param0WriteEnabled=true
    Module.Param1WriteEnabled=false
    Module.Param2WriteEnabled=false
    Module.Param3WriteEnabled=true
   )
   &FN_RibbonMaterial(
    Color=<rgba 0.1788 0.3287 1 1>
    Opacity Scale U=5
    Scale Alpha:NiagaraFloat=&FloatFromCurve(FloatCurve=DataInterface() {propertyValues {Curve {keys [
         {arriveTangent=8.2996, interpMode=RCIM_Cubic, leaveTangent=8.2996, tangentMode=RCTM_User, time=0, value=0}
         {interpMode=RCIM_Cubic, tangentMode=RCTM_User, time=0.2, value=1}
         {tangentMode=RCTM_User, time=1, value=0}
        ]}}})
    Scale Color:Vector3f=&VectorFromFloat(Value=&FloatFromCurve(FloatCurve=DataInterface() {propertyValues {Curve {keys [
          {arriveTangent=7.8976, interpMode=RCIM_Cubic, leaveTangent=7.8976, tangentMode=RCTM_User, time=0, value=0}
          {interpMode=RCIM_Cubic, tangentMode=RCTM_User, time=0.2, value=1}
          {tangentMode=RCTM_User, time=1, value=0}
         ]}}}))
    UV Basis:NiagaraFloat=$Particles.Normalized Spawn Group
   ) {module=FN_RibbonMaterial001}
  ]
  rendererClasses [&NiagaraSpriteRendererProperties, &NiagaraRibbonRendererProperties]
  renderers [
   {rendererClass=&NiagaraSpriteRendererProperties}
   {rendererClass=&NiagaraRibbonRendererProperties, rendererIndex=1}
  ]
 }
 Corona001 {
  emitterUpdateScript [
   &EmitterState(Life Cycle Mode=::NewEnumerator1, Loop Behavior=::NewEnumerator1, Loop Duration=0.1, UseLoopDelay=false)
   &SpawnBurst_Instantaneous(Spawn Count=1, Spawn Time=$User.MinChargeDuration, Use Spawn Probability=false)
  ]
  particleSpawnScript [
   &InitializeParticle(
    Color=<rgba 0.4856 3.1433 25 1>
    Lifetime=0.3
    Module.Write Color=true
    Module.Write Lifetime=true
    Module.Write Mass=true
    Module.Write Material Random=true
    Module.Write Position=true
    Module.Write Scale=false
    Module.Write SpriteRotation=true
    Module.Write SpriteSize=true
    Sprite Rotation=&RandomRangeFloat(Minimum=-180, Maximum=180)
    Sprite Size=&Vector2DFromFloat(Value=75)
   )
   &DynamicMaterialParameters(
    Index 0 Param 1=0.35
    Index 0 Param 2=1
    Module.Param0WriteEnabled=true
    Module.Param1WriteEnabled=true
    Module.Param2WriteEnabled=false
    Module.Param3WriteEnabled=false
   )
   &CameraOffset(Camera Offset Amount=50)
  ]
  particleUpdateScript [
   &ParticleState()
   /SetVariables(Particles.Position=&MakeVector(X=0, Y=0, Z=0)) {enabled=false, isSetParameters=true}
   &ScaleColor(
    Scale Alpha=&FloatFromCurve(FloatCurve=DataInterface() {propertyValues {Curve {keys [
         {arriveTangent=10.5914, interpMode=RCIM_Cubic, time=0, value=1}
         {interpMode=RCIM_Cubic, tangentMode=RCTM_User, time=0.0731, value=1.2166}
         {arriveTangent=-3.4261, interpMode=RCIM_Cubic, leaveTangent=-20281.8965, tangentMode=RCTM_User, time=0.75, value=0}
        ]}}})
    ScaleA=true
    ScaleRGB=false
   )
   /SetVariables(Particles.Renderer2_Color:LinearColor=&ColorFromCurve(ColorCurve=DataInterface() {propertyValues {
       AlphaCurve {keys [<at 0 0.05 RCIM_Cubic>,<at 1 0 RCIM_Cubic>]}
       BlueCurve {keys [<at 0.0078 25 RCIM_Cubic>,<at 0.6767 25>]}
       GreenCurve {keys [<at 0.0078 0 RCIM_Cubic>,<at 0.6767 0>]}
       RedCurve {keys [<at 0.0078 11.3755 RCIM_Cubic>,<at 0.6767 1.9184>]}
      }})) {isSetParameters=true}
   /SetVariables(Particles.Renderer3_Color:LinearColor=&ColorFromCurve(ColorCurve=DataInterface() {propertyValues {
       AlphaCurve {keys [<at 0.3891 0.5 RCIM_Cubic>,<at 1 0 RCIM_Cubic>]}
       BlueCurve {keys [<at 0.602 1 RCIM_Cubic>]}
       GreenCurve {keys [<at 0.602 0 RCIM_Cubic>]}
       RedCurve {keys [<at 0.602 0.1524 RCIM_Cubic>]}
      }})) {isSetParameters=true}
   /SetVariables(
    Particles.CoronaParams1:Vector4f=&MakeVector4(
     X=&Lerp_Float(
      A=0.6
      B=0.05
      Alpha=&FloatFromCurve(
       FloatCurve=DataInterface() {propertyValues {Curve {keys [{arriveTangent=1.4492, interpMode=RCIM_Cubic, leaveTangent=1.4492, tangentMode=RCTM_User, time=0, value=0}, <at 1 1>]}}}
       Scale Curve=0.1
      )
     )
     Y=10
     Z=0.5
     W=1
    )
    Particles.CoronaParams2:Vector4f=&MakeVector4(
     X=&Lerp_Float(
      A=2
      B=8
      Alpha=&FloatFromCurve(FloatCurve=DataInterface() {propertyValues {Curve {keys [{arriveTangent=2.6496, interpMode=RCIM_Cubic, leaveTangent=1.7551, tangentMode=RCTM_User, time=0, value=0}, <at 1 1>]}}})
     )
     Y=&Lerp_Float(A=0, B=0.1, Alpha=$Particles.NormalizedAge)
     Z=2
     W=0
    )
   ) {enabled=false, isSetParameters=true}
   &ScaleSpriteSize(Scale Factor=&Vector2DFromFloat(Value=&FloatFromCurve(FloatCurve=DataInterface() {propertyValues {Curve {keys [
          {arriveTangent=4.3794, interpMode=RCIM_Cubic, leaveTangent=2.141, tangentMode=RCTM_User, time=0, value=0}
          {arriveTangent=0.353, interpMode=RCIM_Cubic, leaveTangent=0.353, tangentMode=RCTM_User, time=1, value=1}
         ]}}})))
   /SetVariables(Particles.Renderer2_Size:Vector2f=&Vector2DFromFloat(Value=&FloatFromCurve(
      FloatCurve=DataInterface() {propertyValues {Curve {keys [
          {arriveTangent=4.3794, interpMode=RCIM_Cubic, leaveTangent=2.141, tangentMode=RCTM_User, time=0, value=0}
          {arriveTangent=0.353, interpMode=RCIM_Cubic, leaveTangent=0.353, tangentMode=RCTM_User, time=1, value=1}
         ]}}}
      Scale Curve=95
     ))) {isSetParameters=true}
   /SetVariables(Particles.Renderer3_Size:Vector2f=&Vector2DFromFloat(Value=&FloatFromCurve(
      FloatCurve=DataInterface() {propertyValues {Curve {keys [
          {arriveTangent=4.3794, interpMode=RCIM_Cubic, leaveTangent=2.141, tangentMode=RCTM_User, time=0, value=0}
          {arriveTangent=0.353, interpMode=RCIM_Cubic, leaveTangent=0.353, tangentMode=RCTM_User, time=1, value=1}
         ]}}}
      Scale Curve=70
     ))) {isSetParameters=true}
   &SpriteRotationRate(Rotation Rate=&Lerp_Float(A=300, B=50, Alpha=$Emitter.InOutCurve)) {enabled=false}
   &FN_Light_Attributes(
    Color=<rgba 1.2153 10.8396 100 1>
    Create Color Attribute=true
    Create Exponent Attribute=true
    Create Radius Attribute=true
    Exponent=6
    Radius=150
    Scale Color=&FloatFromCurve(FloatCurve=DataInterface() {propertyValues {Curve {keys [<at 0 1>,<at 1 0>]}}})
   )
  ]
  rendererClasses [&NiagaraSpriteRendererProperties, &NiagaraLightRendererProperties]
  renderers [
   {rendererClass=&NiagaraSpriteRendererProperties}
   {rendererClass=&NiagaraLightRendererProperties, rendererIndex=1}
   {rendererClass=&NiagaraSpriteRendererProperties, rendererIndex=2}
  ]
 }
 FN_Contrails_Ribbon_Small {
  emitterUpdateScript [
   &EmitterState()
   &FN_SpawnRibbon_v2(Number of Ribbons=5)
  ]
  particleSpawnScript [
   &FN_InitializeRibbon_v3(Lifetime=0.6, Ribbon Width=10) {module=FN_InitializeRibbon_v3001}
   /Niagara/Modules/Spawn/Location/SphereLocation(Hemisphere X=true, MaskSpawnGroup=false, Module.Override Randomness=false, Offset=<10 0 0>, Sphere Radius=7, Surface Only=false)
   &AddVelocityFromPoint(Module.Use Falloff Distance=false, Origin Offset=<9 0 0>, Velocity Strength=350)
   &AddVelocity(Velocity=&MakeVector(X=&RandomRangeFloat(Minimum=0, Maximum=25), Y=0, Z=0)) {enabled=false}
  ]
  particleUpdateScript [
   /Niagara/Modules/Update/Forces/CurlNoiseForce(
    Noise Frequency=10
    Noise Strength=1000
    Pan Noise=true
    Pan Noise Field=<0 0 0.25>
    Random Seed:NiagaraInt32=&RandomRangeInt(Minimum=0, Maximum=1)
   )
   &Drag(Use Linear Drag=true, Use Rotational Drag=false)
   &SolveForcesAndVelocity(Clamp Velocity=false, Limit Acceleration=false)
   &FN_UpdateRibbon(Initial Color:LinearColor=$Particles.Initial.Color, Initial RibbonWidth:NiagaraFloat=$Particles.Initial.RibbonWidth)
   &ScaleRibbonWidth(Ribbon Width Scale=&FloatFromCurve(
     FloatCurve=DataInterface() {propertyValues {Curve {keys [
         {arriveTangent=6.2578, interpMode=RCIM_Cubic, leaveTangent=6.2578, tangentMode=RCTM_User, time=0, value=0}
         <at 0.2 1>
         {arriveTangent=-2.0105, interpMode=RCIM_Cubic, leaveTangent=-2.0105, tangentMode=RCTM_User, time=1, value=0}
        ]}}}
     Scale Curve=1.5
    ))
   &FN_RibbonMaterial(
    Color=&ColorFromCurve(ColorCurve=DataInterface() {propertyValues {
       AlphaCurve {keys [<at 0.0106 1 RCIM_Cubic>,<at 0.9796 1>]}
       BlueCurve {keys [{arriveTangent=-3.673, interpMode=RCIM_Cubic, time=0.0664, value=2}, <at 0.228 1.5>, <at 0.4099 1>, <at 0.5897 1>]}
       GreenCurve {keys [{arriveTangent=0.1965, interpMode=RCIM_Cubic, time=0.0664, value=0.7839}, <at 0.228 0.2737>, <at 0.4099 0.1825>, <at 0.5897 0>]}
       RedCurve {keys [{arriveTangent=0.1301, interpMode=RCIM_Cubic, time=0.0664, value=0.2847}, <at 0.228 0>, <at 0.4099 0>, <at 0.5897 0.5532>]}
      }})
    Scale Alpha:NiagaraFloat=&FloatFromCurve(
     FloatCurve=DataInterface() {propertyValues {Curve {keys [
         {arriveTangent=-0.0061, interpMode=RCIM_Cubic, leaveTangent=-0.0061, tangentMode=RCTM_User, time=0.2, value=1}
         {interpMode=RCIM_Cubic, tangentMode=RCTM_User, time=0.3, value=1}
         {arriveTangent=0.7242, interpMode=RCIM_Cubic, leaveTangent=0.7242, tangentMode=RCTM_User, time=0.8, value=0}
        ]}}}
     Scale Curve=&FloatFromCurve(FloatCurve=DataInterface() {propertyValues {Curve {keys [<at 0 1 RCIM_Cubic>,<at 0.5 0 RCIM_Cubic>]}}}, CurveIndex=$Emitter.Age)
    )
    Scale Color:Vector3f=&VectorFromFloat(Value=&FloatFromCurve(FloatCurve=DataInterface() {propertyValues {Curve {keys [
          {arriveTangent=-0.2994, interpMode=RCIM_Cubic, leaveTangent=-0.2994, tangentMode=RCTM_User, time=0.2, value=1}
          {interpMode=RCIM_Cubic, tangentMode=RCTM_User, time=0.3, value=1}
          {arriveTangent=-0.1169, interpMode=RCIM_Cubic, leaveTangent=-0.1169, tangentMode=RCTM_User, time=0.8, value=0}
         ]}}}))
   ) {module=FN_RibbonMaterial001}
  ]
  rendererClasses [&NiagaraRibbonRendererProperties]
  renderers [{rendererClass=&NiagaraRibbonRendererProperties}]
 }
 Radial_Plasma_Energy_01 {
  emitterUpdateScript [
   &EmitterState()
   &SpawnBurst_Instantaneous(Spawn Count=1, Use Spawn Probability=false)
  ]
  particleSpawnScript [/Niagara/Modules/Spawn/Initialization/V2/InitializeParticle(
    Color Mode=::NewEnumerator0
    Lifetime=0.2
    Sprite Size Mode=::NewEnumerator3
    Uniform Sprite Size=150
    UsePositionOffset=false
    Write Material Random=true
   )]
  particleUpdateScript [
   &ParticleState()
   &Color(
    Color=&ColorFromCurve(ColorCurve=DataInterface() {propertyValues {
       AlphaCurve {keys [{arriveTangent=-1.4022, interpMode=RCIM_Cubic, leaveTangent=-1.4022, tangentMode=RCTM_User, time=0, value=1}, <at 1 0 RCIM_Cubic>]}
       BlueCurve {keys [<at 0.0061 1>,<at 1 4 RCIM_Cubic>]}
       GreenCurve {keys [<at 0.0061 0.5782>,<at 1 0.5253 RCIM_Cubic>]}
       RedCurve {keys [<at 0.0061 0.0138>,<at 1 0.0191 RCIM_Cubic>]}
      }})
    Scale Alpha=&FloatFromCurve(FloatCurve=DataInterface() {propertyValues {Curve {keys [<at 0 0.05>,<at 1 1>]}}})
    Scale Color=&VectorFromFloat(Value=&MakeCustomFloatFromBool(Bool=$User.Hitting, True Float=4, False Float=2))
   )
   &ScaleSpriteSize(Scale Factor=&Vector2DFromFloat(Value=&FloatFromCurve(FloatCurve=DataInterface() {propertyValues {Curve {keys [{arriveTangent=1.3133, interpMode=RCIM_Cubic, leaveTangent=1.3133, tangentMode=RCTM_User, time=0, value=0}, <at 1 1 RCIM_Cubic>]}}}))) {module=ScaleSpriteSize002}
   &FN_Radial_Material(
    Base Sprite Tile U=0.65
    Base Sprite Tile V=2.2
    Depth Fade=20
    Radial Mask Offset X=0.7
    Radial Mask Power=5
    Radial Mask Strength=1.2
    Radial Mask Tile U=0
    Radial Mask Tile V=2
    Sprite Extrusion Speed=2
   )
  ]
  rendererClasses [&NiagaraSpriteRendererProperties]
  renderers [{rendererClass=&NiagaraSpriteRendererProperties}]
 }
 Sparkles {
  emitterUpdateScript [
   &EmitterState(Life Cycle Mode=::NewEnumerator1, Loop Behavior=::NewEnumerator1, Loop Duration=0.1, UseLoopDelay=false)
   &SpawnRate(SpawnRate=75, Use Spawn Probability=false)
   &SpawnBurst_Instantaneous(Spawn Count=12, Use Spawn Probability=false)
  ]
  particleSpawnScript [
   /Niagara/Modules/Spawn/Initialization/InitializeParticle(
    Color=<rgba 0 0.4665 5 1>
    Lifetime=&UniformRangedFloat(Minimum=0.1, Maximum=0.22)
    Mass=&Multiply_Float(
     A=2
     B=/Niagara/DynamicInputs/Math/CalculateParticleRadius(
      Radius Calculation Type:ENiagaraCollisionRadiusOptions=::NewEnumerator0
      Particle Radius:NiagaraFloat=1
      Mesh Dimensions:Vector3f=<1 1 1>
      Method for Calculating Particle Radius:ENiagaraCalculateRadiusOptions=::NewEnumerator2
     )
    )
    Module.Write Color=true
    Module.Write Lifetime=true
    Module.Write Mass=true
    Module.Write Material Random=true
    Module.Write Position=false
    Module.Write Scale=false
    Module.Write SpriteRotation=false
    Module.Write SpriteSize=true
    Position=&SimulationPosition()
    Sprite Size=&Vector2DFromFloat(Value=&UniformRangedFloat(Minimum=2.5, Maximum=3.5))
   )
   &SphereLocation(Hemisphere X=true, MaskSpawnGroup=false, Module.Override Randomness=false, Offset=<10 0 0>, Sphere Radius=7, Surface Only=false)
   &AddVelocityFromPoint(Module.Use Falloff Distance=false, Origin Offset=<10 0 0>, Velocity Strength=1000)
   &DynamicMaterialParameters(
    Index 0 Param 1=&UniformRangedFloat(Minimum=0.02, Maximum=0.1)
    Index 0 Param 2=&UniformRangedFloat(Minimum=2, Maximum=3)
    Module.Param0WriteEnabled=true
    Module.Param1WriteEnabled=true
    Module.Param2WriteEnabled=false
    Module.Param3WriteEnabled=false
   )
   &AddVelocity(Velocity=&MakeVector(X=&RandomRangeFloat(Minimum=0, Maximum=2000), Y=0, Z=0))
   /SetVariables(Particles.Renderer2_Color:LinearColor=<rgba 0.5616 0 1 0.5>) {isSetParameters=true}
  ]
  particleUpdateScript [
   &ParticleState()
   &Drag(Drag=&RandomRangeFloat(Minimum=2, Maximum=5), Use Linear Drag=true, Use Rotational Drag=false)
   &SolveForcesAndVelocity(Clamp Velocity=false, Limit Acceleration=false)
   &ScaleSpriteSizeBySpeed(
    Max Scale Factor=<1 7>
    Sample Scale Factor By Curve=true
    Scale Factor Curve=DataInterface() {propertyValues {Curve {keys [
        {arriveTangent=2.468, interpMode=RCIM_Cubic, leaveTangent=0.1409, tangentMode=RCTM_User, time=0, value=0}
        {arriveTangent=2.839, interpMode=RCIM_Cubic, leaveTangent=2.839, tangentMode=RCTM_User, time=1, value=1}
       ]}}}
    Velocity Threshold=1500
   )
   &ScaleColor(
    Scale RGB=&VectorFromFloat(Value=&FloatFromCurve(FloatCurve=DataInterface() {propertyValues {Curve {keys [
          {arriveTangent=1.8182, interpMode=RCIM_Cubic, time=0.1, value=2}
          {interpMode=RCIM_Cubic, tangentMode=RCTM_User, time=0.5, value=1}
          {arriveTangent=-3.4105, interpMode=RCIM_Cubic, leaveTangent=-4.1755, tangentMode=RCTM_User, time=1, value=0}
         ]}}}))
    ScaleA=true
    ScaleRGB=true
   )
   &ScaleSpriteSize(Scale Factor=&Vector2DFromFloat(Value=&FloatFromCurve(FloatCurve=DataInterface() {propertyValues {Curve {keys [
          {arriveTangent=7.0713, interpMode=RCIM_Cubic, leaveTangent=7.0713, tangentMode=RCTM_User, time=-0.0069, value=0.25}
          {interpMode=RCIM_Cubic, tangentMode=RCTM_User, time=0.2, value=1}
          {interpMode=RCIM_Cubic, leaveTangent=-3.0611, time=1, value=0}
         ]}}})))
   /SetVariables(Particles.Renderer2_Size:Vector2f=&Multiply_Vector2(A=$Particles.SpriteSize, B=<6 1.35>)) {isSetParameters=true}
  ]
  rendererClasses [&NiagaraSpriteRendererProperties]
  renderers [{rendererClass=&NiagaraSpriteRendererProperties}]
 }
}
summary {
 systemName=NS_BadNew_Muzzle
 emitters [
  {emitterName=Sparkles, rendererClasses [&NiagaraSpriteRendererProperties]}
  {emitterName=Corona001, rendererClasses [&NiagaraSpriteRendererProperties, &NiagaraLightRendererProperties]}
  {emitterName=Radial_Plasma_Energy_01, rendererClasses [&NiagaraSpriteRendererProperties]}
  {emitterName=FN_Contrails_Ribbon_Small, rendererClasses [&NiagaraRibbonRendererProperties]}
  {
   emitterName=BoneFlow_Fixed
   rendererClasses [&NiagaraSpriteRendererProperties, &NiagaraRibbonRendererProperties]
  }
 ]
 userVariables [
  User.ChargeUpLength {type=NiagaraFloat, defaultValue=0, description=""}
  User.Color_Tint {type=LinearColor, description="", defaultValue {struct=LinearColor, value=<rgba 1 1 1 1>}}
  User.EmissiveIntensity {type=NiagaraFloat, defaultValue=0, description=""}
  User.Hitting {type=NiagaraBool}
  User.MinChargeDuration {type=NiagaraFloat, defaultValue=0, description=""}
  User.RibbonIDOffset {type=NiagaraInt32}
  User.SkeletalMesh {
   type=NiagaraDataInterfaceSkeletalMesh
   defaultValue {
    struct=NiagaraExt_VariableValue_DataInterface
    value {dataInterface=&self:NiagaraDataInterfaceSkeletalMesh_0}
   }
  }
 ]
}
