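# Prototype Niagara effect for the BadNews projectile: a pulsing plasma orb with glowing core, sphere meshes, radial sprites, floating particles and arcing tendrils.
# System-level growth, shrink, waveform and timer values drive size and flicker, while User.Hitting boosts the core, light and sprites.
# Tendril ribbons read TendrilsEndPoints particles on a surrounding sphere, forming Hermite splines with waveform wobble.
# Debug, Large_Sphere, standalone Light and one radial emitter are disabled.
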NiagaraSystem /MotherGameplay/Items/BadNews/ProtoAssets/NS_BadNews_Projectile
dependencies {
 usedDataInterfaces [
  &NiagaraDataInterfaceColorCurve
  &NiagaraDataInterfaceCurve
  &NiagaraDataInterfaceArrayColor
  &NiagaraDataInterfaceParticleRead
  &NiagaraDataInterfaceArrayFloat
 ]
 usedDynamicInputs [&Lerp_Vector2]
 usedModules [
  &SystemState
  &self:SystemSpawnScript.SystemScriptSource.SystemScriptGraph.NiagaraNodeAssignment_0.SetVariables_#1
  &self:SystemSpawnScript.SystemScriptSource.SystemScriptGraph.NiagaraNodeAssignment_1.SetVariables_#2
  &FN_Timer
  &self:SystemSpawnScript.SystemScriptSource.SystemScriptGraph.NiagaraNodeAssignment_2.SetVariables_#3
  &self:SystemSpawnScript.SystemScriptSource.SystemScriptGraph.NiagaraNodeAssignment_3.SetVariables_#4
 ]
}
emitters {
 Debug {
  bEnabled=false
  emitterUpdateScript [
   &EmitterState()
   &SpawnBurst_Instantaneous(Spawn Count=1, Use Spawn Probability=false)
  ]
  particleSpawnScript [&InitializeParticle(
    Color Mode=::NewEnumerator0
    Lifetime=$User.ChargeUpLength
    Mesh Scale Mode=::NewEnumerator3
    UsePositionOffset=false
    Write Material Random=true
   )]
  particleUpdateScript [
   &ParticleState(Kill Particles When Lifetime Has Elapsed=false, Let Infinitely Lived Particles Die When Emitter Deactivates=true)
   &ScaleMeshSize(Scale Factor=&VectorFromFloat(Value=&Multiply_Float(A=$User.FX_Sphere_Small_Size, B=2)))
  ]
  rendererClasses [&NiagaraMeshRendererProperties]
  renderers [{rendererClass=&NiagaraMeshRendererProperties, rendererIndex=0}]
 }
 Floating_Plasma_Particles {
  emitterUpdateScript [
   &EmitterState()
   &SpawnRate(
    SpawnRate=&FloatFromCurve(
     FloatCurve=DataInterface() {propertyValues {Curve {keys [{tangentMode=RCTM_User, time=0, value=0}, {tangentMode=RCTM_User, time=1, value=1}]}}}
     CurveIndex=$Emitter.Age
     Scale Curve=100
    )
    Use Spawn Probability=false
   )
  ]
  particleSpawnScript [
   &FN_SetSpawnGroup()
   &InitializeParticle(
    Color Mode=::NewEnumerator0
    Lifetime=0.5
    Sprite Size Mode=::NewEnumerator3
    Uniform Sprite Size=&Multiply_Float(A=4, B=$System.Shrinking)
    UsePositionOffset=false
    Write Material Random=true
   )
   &SphereLocation(
    MaskSpawnGroup=false
    Override Seed=true
    Random Seed:NiagaraInt32=&ReturnParticleIDAttributes()
    Randomness Mode:ENiagaraRandomnessMode=::NewEnumerator1
    Sphere Coordinate Space=::Simulation
    Sphere Radius=&Multiply_Float(A=$System.GrowingWOffset, B=$System.Shrinking)
    Surface Only=false
    UseNonUniformScale=false
    UseOffset=false
    UseSphereOrientation=false
   )
   &AddVelocity(Velocity=<-150 0 0>) {enabled=false}
  ]
  particleUpdateScript [
   &ParticleState()
   &Color(
    Color=&SelectLinearColorFromArray(
     Array Sampling Mode=::NewEnumerator1
     Direct Array Index=$Particles.SpawnGroup
     Color Selection Array=DataInterface() {propertyValues {ColorData [<rgba 0 15.3053 50 1>, <rgba 50 0 18.2953 1>, <rgba 37.1678 0 50 1>, <rgba 0 50 21.3067 1>]}}
     Random Seed=?
    )
    Scale Alpha=&FloatFromCurve(FloatCurve=DataInterface() {propertyValues {Curve {keys [<at 0 0.05>,<at 1 1>]}}})
   )
   &ScaleSpriteSize(Scale Factor=&Vector2DFromFloat(Value=&FloatFromCurve(FloatCurve=DataInterface() {propertyValues {Curve {keys [
          {interpMode=RCIM_Cubic, tangentMode=RCTM_User, time=0, value=0}
          {interpMode=RCIM_Cubic, tangentMode=RCTM_User, time=0.5, value=1}
          {tangentMode=RCTM_User, time=1, value=0}
         ]}}}))) {enabled=false}
   /SetVariables(Particles.SpriteSize=&Multiply_Vector2(
     A=$Particles.SpriteSize
     B=&Vector2DFromFloat(Value=&FloatFromCurve(
       FloatCurve=DataInterface() {propertyValues {Curve {keys [
           {arriveTangent=-0.0977, interpMode=RCIM_Cubic, leaveTangent=-0.0977, tangentMode=RCTM_User, time=0, value=0.1}
           {arriveTangent=0.5884, interpMode=RCIM_Cubic, leaveTangent=0.5884, tangentMode=RCTM_User, time=1, value=1}
          ]}}}
       CurveIndex=$Emitter.Age
      ))
    )) {enabled=false, isSetParameters=true}
   &SolveForcesAndVelocity(Clamp Velocity=false, Limit Acceleration=false)
   &ScaleSpriteSize(Scale Factor=&Vector2DFromFloat(Value=&MakeCustomFloatFromBool(Bool=$User.Hitting, True Float=10, False Float=1))) {enabled=false, module=ScaleSpriteSize001}
   &ScaleSpriteSize(
    Scale Sprite Size Mode=::NewEnumerator0
    Uniform Scale Factor=&Multiply_Float(
     A=&Multiply_Float(
      A=&FloatFromCurve(FloatCurve=DataInterface() {propertyValues {Curve {keys [
           {interpMode=RCIM_Cubic, tangentMode=RCTM_User, time=0, value=0}
           {interpMode=RCIM_Cubic, tangentMode=RCTM_User, time=0.5, value=1}
           {tangentMode=RCTM_User, time=1, value=0}
          ]}}})
      B=&FloatFromCurve(
       FloatCurve=DataInterface() {propertyValues {Curve {keys [
           {arriveTangent=-0.0977, interpMode=RCIM_Cubic, leaveTangent=-0.0977, tangentMode=RCTM_User, time=0, value=0.1}
           {arriveTangent=0.5884, interpMode=RCIM_Cubic, leaveTangent=0.5884, tangentMode=RCTM_User, time=1, value=1}
          ]}}}
       CurveIndex=$Emitter.Age
      )
     )
     B=&MakeCustomFloatFromBool(Bool=$User.Hitting, True Float=10, False Float=1)
    )
   ) {module=ScaleSpriteSize002}
  ]
  rendererClasses [&NiagaraSpriteRendererProperties]
  renderers [{rendererClass=&NiagaraSpriteRendererProperties}]
 }
 Large_Sphere {
  bEnabled=false
  emitterUpdateScript [
   &EmitterState()
   &SpawnBurst_Instantaneous(Spawn Count=1, Use Spawn Probability=false)
  ]
  particleSpawnScript [
   &InitializeParticle(
    Color=<rgba 2 2 2 0.05>
    Lifetime=$User.ChargeUpLength
    Mesh Scale Mode=::NewEnumerator3
    Mesh Uniform Scale=0.0157
    UsePositionOffset=false
    Write Material Random=true
   )
   &InitialMeshOrientation(Orientation Vector=<0 0 1>, Use Orientation Vector=true, Use Rotation Vector=false)
  ]
  particleUpdateScript [
   &ParticleState(Kill Particles When Lifetime Has Elapsed=false, Let Infinitely Lived Particles Die When Emitter Deactivates=true)
   /SetVariables(Particles.BFGTimer:NiagaraFloat=$Engine.DeltaTime) {isSetParameters=true, module=SetVariables_#5}
   &ScaleMeshSize(Scale Factor=&VectorFromFloat(Value=&FloatFromCurve(
      FloatCurve=DataInterface() {propertyValues {Curve {keys [
          {interpMode=RCIM_Cubic, leaveTangent=0.6233, tangentMode=RCTM_Break, time=0, value=0}
          {interpMode=RCIM_Cubic, tangentMode=RCTM_Break, time=5, value=1}
         ]}}}
      CurveIndex=$Emitter.Age
      Scale Curve=&Lerp_Float(
       A=1
       B=&Lerp_Float(
        A=0.8
        B=1
        Alpha=&Waveform(
         "Waveform [1]"=::NewEnumerator0
         "[1] Frequency"=&Lerp_Float(A=7, B=11, Alpha=$System.FN_Timer.Value)
         "[1] Amplitude Min/Max"=<0 1>
         "[2] Phase"=$Emitter.Age
         "[3] Phase"=$Emitter.Age
        )
       )
       Alpha=$System.FN_Timer.Value
      )
     ))) {enabled=false}
   &DynamicMaterialParameters(
    Index 0 Param 2=HlslExpression() {hlslExpression=Emitter.Age*.2}
    Index 0 Param 3=2
    Index 0 Param 4=2
    Index 1 Param 1=2
    Index 1 Param 2=1
    Index 1 Param 3=0.7
    Index 1 Param 4=1
    Index 2 Param 2=HlslExpression() {hlslExpression=Emitter.Age*.2}
    Index 2 Param 3=3
    Index 2 Param 4=3
    Module.Param0WriteEnabled=true
    Module.Param0WriteEnabled1=true
    Module.Param0WriteEnabled2=true
    Module.Param1WriteEnabled=true
    Module.Param1WriteEnabled1=true
    Module.Param1WriteEnabled2=true
    Module.Param2WriteEnabled=true
    Module.Param2WriteEnabled1=true
    Module.Param2WriteEnabled2=true
    Module.Param3WriteEnabled=true
    Module.Param3WriteEnabled1=true
    Module.Param3WriteEnabled2=true
    Write Parameter Index 1=true
    Write Parameter Index 2=true
   ) {module=DynamicMaterialParameters001}
   &ScaleMeshSize(Scale Factor=&VectorFromFloat(Value=$User.FX_Sphere_Large_Size)) {module=ScaleMeshSize001}
  ]
  rendererClasses [&NiagaraMeshRendererProperties]
  renderers [{rendererClass=&NiagaraMeshRendererProperties, rendererIndex=0}]
 }
 Light {
  bEnabled=false
  emitterUpdateScript [
   &EmitterState(Life Cycle Mode=::NewEnumerator1, Loop Behavior=::NewEnumerator1, Loop Duration Mode=::NewEnumerator1, UseLoopDelay=false)
   &SpawnBurst_Instantaneous(Spawn Count=1, Use Spawn Probability=false)
  ]
  particleSpawnScript [&InitializeParticle(Color=<rgba 0 5.4181 20 1>, Lifetime=5, UsePositionOffset=false, Write Material Random=true)]
  particleUpdateScript [
   &ParticleState(Kill Particles When Lifetime Has Elapsed=false, Let Infinitely Lived Particles Die When Emitter Deactivates=true)
   /SetVariables(
    Particles.Attentuation Radius:NiagaraFloat=&Lerp_Float(A=220, B=300, Alpha=$System.WaveformData)
    Particles.Light Intensity:NiagaraFloat=&Lerp_Float(A=10, B=30, Alpha=$System.WaveformData)
    Particles.PointLightColor:LinearColor=&LerpLinearColors(StartColor=<rgba 0 0.1923 3 1>, EndColor=<rgba 0 0.2741 10 1>, LerpFactor=$System.WaveformData)
   ) {isSetParameters=true}
   /SetVariables(
    Particles.LightRendererColor:LinearColor=&MultiplyLinearColor(
     Linear Color=&LerpLinearColors(
      StartColor=&LerpLinearColors(
       StartColor=<rgba 0 7.1401 20 1>
       EndColor=<rgba 0.2223 0 10 1>
       LerpFactor=&RemapRange(Input Value To Remap=$System.WaveformData, Input Min=0, Input Max=2, Output Min=0, Output Max=1)
      )
      EndColor=&LerpLinearColors(
       StartColor=<rgba 0 10.7101 30 1>
       EndColor=<rgba 0.3335 0 15 1>
       LerpFactor=&RemapRange(Input Value To Remap=$System.WaveFormFastData, Input Min=0, Input Max=2, Output Min=0, Output Max=1)
      )
      LerpFactor=$System.FN_Timer001.Value
     )
     Scale Factor=&MakeVector4_FromFloat(Float=&MakeCustomFloatFromBool(Bool=$User.Hitting, True Float=2, False Float=1))
    )
    Particles.LightRendererRadiusScale:NiagaraFloat=&Multiply_Float(
     A=&Lerp_Float(A=32, B=40, Alpha=$System.WaveformData)
     B=&MakeCustomFloatFromBool(Bool=$User.Hitting, True Float=2, False Float=1)
    )
   ) {isSetParameters=true}
  ]
  rendererClasses [&NiagaraComponentRendererProperties, &NiagaraLightRendererProperties]
  renderers [
   {rendererClass=&NiagaraComponentRendererProperties, rendererIndex=0}
   {rendererClass=&NiagaraLightRendererProperties, rendererIndex=1}
  ]
 }
 Plasma_Center_Energy {
  emitterUpdateScript [
   &EmitterState()
   &SpawnBurst_Instantaneous(Spawn Count=1, Use Spawn Probability=false)
  ]
  particleSpawnScript [&InitializeParticle(
    Color Mode=::NewEnumerator0
    Lifetime=$User.ChargeUpLength
    Sprite Size Mode=::NewEnumerator3
    Uniform Sprite Size=2
    UsePositionOffset=false
    Write Material Random=true
   )]
  particleUpdateScript [
   &ParticleState(Kill Particles When Lifetime Has Elapsed=false, Let Infinitely Lived Particles Die When Emitter Deactivates=true)
   &Color(
    Color=<rgba 0 19.621 50 1>
    Scale Alpha=&FloatFromCurve(FloatCurve=DataInterface() {propertyValues {Curve {keys [<at 0 0.05>,<at 1 1>]}}})
    Scale Color=&VectorFromFloat(Value=&MakeCustomFloatFromBool(Bool=$User.Hitting, True Float=2, False Float=1))
   )
   &ScaleSpriteSize(Scale Factor=&Vector2DFromFloat(Value=$System.GrowingWOffset)) {enabled=false, module=ScaleSpriteSize002}
   &ScaleSpriteSize(Scale Factor=&Vector2DFromFloat(Value=&MakeCustomFloatFromBool(Bool=$User.Hitting, True Float=1.5, False Float=1))) {enabled=false, module=ScaleSpriteSize001}
   &ScaleSpriteSize(Scale Factor=&Vector2DFromFloat(Value=&Lerp_Float(A=0.8, B=1, Alpha=$System.WaveformData)))
   &FN_Light_Attributes(
    Color=<rgba 0 19.621 50 1>
    Create Color Attribute=true
    Create Exponent Attribute=true
    Create Radius Attribute=true
    Exponent=5
    Radius=&Multiply_Float(A=0.8, B=$User.FX_Sphere_Small_Size)
    Scale Color=&Multiply_Float(A=0.005, B=$User.FX_Sphere_Small_Size)
   ) {enabled=false}
   &FN_Light_Attributes(
    Color=&LerpLinearColors(
     StartColor=&LerpLinearColors(
      StartColor=<rgba 0 7.1401 20 1>
      EndColor=<rgba 0.2223 0 10 1>
      LerpFactor=&RemapRange(Input Value To Remap=$System.WaveformData, Input Min=0, Input Max=2, Output Min=0, Output Max=1)
     )
     EndColor=&LerpLinearColors(
      StartColor=<rgba 0 10.7101 30 1>
      EndColor=<rgba 0.3335 0 15 1>
      LerpFactor=&RemapRange(Input Value To Remap=$System.WaveFormFastData, Input Min=0, Input Max=2, Output Min=0, Output Max=1)
     )
     LerpFactor=$System.FN_Timer001.Value
    )
    Create Color Attribute=true
    Create Exponent Attribute=true
    Create Radius Attribute=true
    Exponent=5
    Radius=&Multiply_Float(
     A=$System.Shrinking
     B=&Multiply_Float(
      A=&Multiply_Float(
       A=0.05
       B=&Multiply_Float(
        A=&Lerp_Float(A=32, B=40, Alpha=$System.WaveformData)
        B=&MakeCustomFloatFromBool(Bool=$User.Hitting, True Float=2, False Float=1)
       )
      )
      B=$System.GrowingWOffset
     )
    )
    Scale Color=&Multiply_Float(
     A=&Multiply_Float(A=0.005, B=&MakeCustomFloatFromBool(Bool=$User.Hitting, True Float=2, False Float=1))
     B=$User.FX_Sphere_Small_Size
    )
   ) {module=FN_Light_Attributes001}
   /SetVariables(
    Particles.LightRendererColor:LinearColor=&MultiplyLinearColor(
     Linear Color=&LerpLinearColors(
      StartColor=&LerpLinearColors(
       StartColor=<rgba 0 7.1401 20 1>
       EndColor=<rgba 0.2223 0 10 1>
       LerpFactor=&RemapRange(Input Value To Remap=$System.WaveformData, Input Min=0, Input Max=2, Output Min=0, Output Max=1)
      )
      EndColor=&LerpLinearColors(
       StartColor=<rgba 0 10.7101 30 1>
       EndColor=<rgba 0.3335 0 15 1>
       LerpFactor=&RemapRange(Input Value To Remap=$System.WaveFormFastData, Input Min=0, Input Max=2, Output Min=0, Output Max=1)
      )
      LerpFactor=$System.FN_Timer001.Value
     )
     Scale Factor=&MakeVector4_FromFloat(Float=&MakeCustomFloatFromBool(Bool=$User.Hitting, True Float=2, False Float=1))
    )
    Particles.LightRendererRadiusScale:NiagaraFloat=&Multiply_Float(
     A=&Lerp_Float(A=32, B=40, Alpha=$System.WaveformData)
     B=&MakeCustomFloatFromBool(Bool=$User.Hitting, True Float=2, False Float=1)
    )
   ) {enabled=false, isSetParameters=true}
   &ScaleSpriteSize(Scale Factor=&Vector2DFromFloat(Value=$System.Shrinking)) {enabled=false, module=ScaleSpriteSize003}
   &ScaleSpriteSize(
    Scale Sprite Size Mode=::NewEnumerator0
    Uniform Scale Factor=&Multiply_Float(
     A=&Multiply_Float(
      A=$System.GrowingWOffset
      B=&MakeCustomFloatFromBool(Bool=$User.Hitting, True Float=1.5, False Float=1)
     )
     B=&Multiply_Float(A=&Lerp_Float(A=0.8, B=1, Alpha=$System.WaveformData), B=$System.Shrinking)
    )
   ) {module=ScaleSpriteSize004}
  ]
  rendererClasses [&NiagaraSpriteRendererProperties, &NiagaraLightRendererProperties]
  renderers [
   {rendererClass=&NiagaraSpriteRendererProperties}
   {rendererClass=&NiagaraSpriteRendererProperties, rendererIndex=1}
   {rendererClass=&NiagaraLightRendererProperties, rendererIndex=2}
  ]
 }
 Radial_Plasma_Energy_01 {
  emitterUpdateScript [
   &EmitterState()
   &SpawnBurst_Instantaneous(Spawn Count=1, Use Spawn Probability=false)
  ]
  particleSpawnScript [&InitializeParticle(
    Color Mode=::NewEnumerator0
    Lifetime=$User.ChargeUpLength
    Sprite Size Mode=::NewEnumerator3
    Uniform Sprite Size=1.5
    UsePositionOffset=false
    Write Material Random=true
   )]
  particleUpdateScript [
   &ParticleState(Kill Particles When Lifetime Has Elapsed=false, Let Infinitely Lived Particles Die When Emitter Deactivates=true)
   &FN_Radial_Material(Base Sprite Tile U=1.5, Depth Fade=200, Radial Mask Power=2.5, Sprite Extrusion Speed=0.5)
   &Color(
    Color=&ColorFromCurve(ColorCurve=DataInterface() {propertyValues {
       AlphaCurve {keys [<at 0 1 RCIM_Cubic>,<at 1 1 RCIM_Cubic>]}
       BlueCurve {keys [<at 0 1>, {arriveTangent=3, interpMode=RCIM_Cubic, leaveTangent=3, time=0, value=1}, <at 1 4 RCIM_Cubic>]}
       GreenCurve {keys [<at 0 1>, {arriveTangent=-1, interpMode=RCIM_Cubic, leaveTangent=-1, time=0, value=0}, <at 1 0 RCIM_Cubic>]}
       RedCurve {keys [<at 0 0.8353>, {arriveTangent=0.8441, interpMode=RCIM_Cubic, leaveTangent=0.8441, time=0, value=0.1441}, <at 1 1.6795 RCIM_Cubic>]}
      }})
    Scale Alpha=&FloatFromCurve(FloatCurve=DataInterface() {propertyValues {Curve {keys [<at 0 0.05>,<at 1 1>]}}})
    Scale Color=&VectorFromFloat(Value=&MakeCustomFloatFromBool(Bool=$User.Hitting, True Float=4, False Float=2))
   )
   &ScaleSpriteSize(Scale Factor=&Vector2DFromFloat(Value=$System.GrowingWOffset)) {enabled=false}
   &ScaleSpriteSize(Scale Factor=&Vector2DFromFloat(Value=&MakeCustomFloatFromBool(Bool=$User.Hitting, True Float=1.5, False Float=1))) {enabled=false, module=ScaleSpriteSize001}
   &ScaleSpriteSize(Scale Factor=&Vector2DFromFloat(Value=$System.Shrinking)) {enabled=false, module=ScaleSpriteSize002}
   &ScaleSpriteSize(
    Scale Sprite Size Mode=::NewEnumerator0
    Uniform Scale Factor=&Multiply_Float(
     A=$System.GrowingWOffset
     B=&Multiply_Float(A=&MakeCustomFloatFromBool(Bool=$User.Hitting, True Float=1.5, False Float=1), B=$System.Shrinking)
    )
   ) {module=ScaleSpriteSize003}
  ]
  rendererClasses [&NiagaraSpriteRendererProperties]
  renderers [{rendererClass=&NiagaraSpriteRendererProperties}]
 }
 Radial_Plasma_Energy_01001 {
  bEnabled=false
  emitterUpdateScript [
   &EmitterState(Inactive Response=::NewEnumerator1, Life Cycle Mode=::NewEnumerator1, Loop Duration=1, UseLoopDelay=false)
   &SpawnRate(SpawnRate=20, Use Spawn Probability=false)
  ]
  particleSpawnScript [
   /SetVariables(Particles.VisibilityTag=&RandomRangeInt(Minimum=1, Maximum=4)) {isSetParameters=true}
   &FN_SetSpawnGroup(Number of Spawn Groups=2)
   &InitializeParticle(
    Color Mode=::NewEnumerator0
    Lifetime=&RandomRangeFloat(Minimum=0.35, Maximum=1)
    Sprite Rotation Angle=&RandomRangeFloat(Minimum=0, Maximum=360)
    Sprite Rotation Mode=::NewEnumerator0
    Sprite Size Mode=::NewEnumerator1
    Uniform Sprite Size Max=550
    Uniform Sprite Size Min=142
    UsePositionOffset=false
    Write Material Random=true
   )
  ]
  particleUpdateScript [
   &ParticleState()
   &FN_Radial_Material(
    Depth Fade=25
    Radial Mask Power=2.5
    Sprite Extrusion Speed=1
    Sprite Spin Speed:NiagaraFloat=&RandomRangeFloat(Minimum=-0.2, Maximum=0.2)
   )
   &Color(
    Color=&MakeCustomLinearColorFromBool(Bool=HlslExpression() {hlslExpression="Particles.SpawnGroup == 0"}, True Color=<rgba 0 1.3347 5 1>, False Color=<rgba 1.7932 0 5 1>)
    Scale Alpha=&FloatFromCurve(FloatCurve=DataInterface() {propertyValues {Curve {keys [{interpMode=RCIM_Cubic, tangentMode=RCTM_User, time=0, value=1}, {interpMode=RCIM_Cubic, tangentMode=RCTM_User, time=1, value=0}]}}})
    Scale Color=&VectorFromFloat(Value=2)
   )
   &ScaleSpriteSize(Scale Factor=&Vector2DFromFloat(Value=&FloatFromCurve(
      FloatCurve=DataInterface() {propertyValues {Curve {keys [
          {arriveTangent=0.7137, interpMode=RCIM_Cubic, leaveTangent=0.7137, tangentMode=RCTM_User, time=0, value=0}
          {arriveTangent=0.059, interpMode=RCIM_Cubic, leaveTangent=0.059, tangentMode=RCTM_User, time=1, value=1}
         ]}}}
      Scale Curve=&FloatFromCurve(
       FloatCurve=DataInterface() {propertyValues {Curve {keys [
           {interpMode=RCIM_Cubic, leaveTangent=0.6233, tangentMode=RCTM_Break, time=0, value=0}
           {interpMode=RCIM_Cubic, tangentMode=RCTM_Break, time=5, value=1}
          ]}}}
       CurveIndex=$Emitter.Age
      )
     )))
  ]
  rendererClasses [&NiagaraSpriteRendererProperties]
  renderers [
   {rendererClass=&NiagaraSpriteRendererProperties}
   {rendererClass=&NiagaraSpriteRendererProperties, rendererIndex=1}
   {rendererClass=&NiagaraSpriteRendererProperties, rendererIndex=2}
   {rendererClass=&NiagaraSpriteRendererProperties, rendererIndex=3}
  ]
 }
 Radial_Plasma_Energy_02 {
  emitterUpdateScript [
   &EmitterState()
   &SpawnBurst_Instantaneous(Spawn Count=1, Use Spawn Probability=false)
  ]
  particleSpawnScript [&InitializeParticle(
    Color Mode=::NewEnumerator0
    Lifetime=$User.ChargeUpLength
    Sprite Size Mode=::NewEnumerator3
    Uniform Sprite Size=6
    UsePositionOffset=false
    Write Material Random=true
   )]
  particleUpdateScript [
   &ParticleState(Kill Particles When Lifetime Has Elapsed=false, Let Infinitely Lived Particles Die When Emitter Deactivates=true)
   &FN_Radial_Material(
    Base Sprite Tile V=2
    Depth Fade=250
    Radial Mask Offset X=0.7
    Radial Mask Power=10
    Radial Mask Tile U=0
    Radial Mask Tile V=2
    Sprite Extrusion Speed=0.6
    Sprite Spin Speed=0.1
   )
   &Color(
    Color=&LerpLinearColors(
     StartColor=<rgba 0 0.6691 2 1>
     EndColor=<rgba 2 0 1.905 1>
     LerpFactor=&Waveform("Waveform [1]"=::NewEnumerator0, "[1] Frequency"=0.5, "[1] Amplitude Min/Max"=<0 1>, "[2] Phase"=$Emitter.Age, "[3] Phase"=$Emitter.Age)
    )
    Scale Alpha=0.7
    Scale Color=&VectorFromFloat(Value=&MakeCustomFloatFromBool(Bool=$User.Hitting, True Float=2, False Float=1))
   )
   &ScaleSpriteSize(Scale Factor=&Vector2DFromFloat(Value=$System.GrowingWOffset)) {enabled=false, module=ScaleSpriteSize001}
   &ScaleSpriteSize(Scale Factor=&Vector2DFromFloat(Value=$System.Shrinking)) {enabled=false, module=ScaleSpriteSize002}
   &ScaleSpriteSize(Scale Factor=&Vector2DFromFloat(Value=&Lerp_Float(
      A=1
      B=&Lerp_Float(
       A=0.8
       B=1
       Alpha=&Waveform("Waveform [1]"=::NewEnumerator0, "[1] Frequency"=7, "[1] Amplitude Min/Max"=<0 1>, "[2] Phase"=$Emitter.Age, "[3] Phase"=$Emitter.Age)
      )
      Alpha=$System.FN_Timer.Value
     ))) {enabled=false}
   &ScaleSpriteSize(
    Scale Sprite Size Mode=::NewEnumerator0
    Uniform Scale Factor=&Multiply_Float(
     A=&Multiply_Float(A=$System.GrowingWOffset, B=$System.Shrinking)
     B=&Lerp_Float(
      A=1
      B=&Lerp_Float(
       A=0.8
       B=1
       Alpha=&Waveform("Waveform [1]"=::NewEnumerator0, "[1] Frequency"=7, "[1] Amplitude Min/Max"=<0 1>, "[2] Phase"=$Emitter.Age, "[3] Phase"=$Emitter.Age)
      )
      Alpha=$System.FN_Timer.Value
     )
    )
   ) {module=ScaleSpriteSize003}
  ]
  rendererClasses [&NiagaraSpriteRendererProperties]
  renderers [{rendererClass=&NiagaraSpriteRendererProperties}]
 }
 Small_Sphere {
  emitterUpdateScript [
   &EmitterState()
   &SpawnBurst_Instantaneous(Spawn Count=1, Use Spawn Probability=false)
  ]
  particleSpawnScript [
   &InitializeParticle(
    Color=<rgba 2 2 2 1>
    Lifetime=$User.ChargeUpLength
    Mesh Scale Mode=::NewEnumerator3
    Mesh Uniform Scale=0.015
    UsePositionOffset=false
    Write Material Random=true
   )
   &InitialMeshOrientation(Orientation Vector=<0 0 1>, Use Orientation Vector=true, Use Rotation Vector=false)
  ]
  particleUpdateScript [
   &ParticleState(Kill Particles When Lifetime Has Elapsed=false, Let Infinitely Lived Particles Die When Emitter Deactivates=true)
   /SetVariables(Particles.BFGTimer:NiagaraFloat=$Engine.DeltaTime) {isSetParameters=true, module=SetVariables_#5}
   &DynamicMaterialParameters(
    Index 0 Param 2=HlslExpression() {hlslExpression=Emitter.Age*.2}
    Index 0 Param 3=1
    Index 0 Param 4=1
    Index 1 Param 1=1
    Index 1 Param 2=1
    Index 1 Param 3=1
    Index 1 Param 4=1
    Index 2 Param 2=HlslExpression() {hlslExpression=Emitter.Age*.2}
    Index 2 Param 3=2
    Index 2 Param 4=2
    Module.Param0WriteEnabled=true
    Module.Param0WriteEnabled1=true
    Module.Param0WriteEnabled2=true
    Module.Param1WriteEnabled=true
    Module.Param1WriteEnabled1=true
    Module.Param1WriteEnabled2=true
    Module.Param2WriteEnabled=true
    Module.Param2WriteEnabled1=true
    Module.Param2WriteEnabled2=true
    Module.Param3WriteEnabled=true
    Module.Param3WriteEnabled1=true
    Module.Param3WriteEnabled2=true
    Write Parameter Index 1=true
    Write Parameter Index 2=true
   ) {module=DynamicMaterialParameters001}
   /SetVariables(Particles.PinkPlasmaOrbColor:LinearColor=<rgba 0 0 0 0.7>) {isSetParameters=true}
   &ScaleMeshSize(Scale Factor=&VectorFromFloat(Value=$System.GrowingWOffset)) {enabled=false, module=ScaleMeshSize001}
   &ScaleMeshSize(Scale Factor=&VectorFromFloat(Value=$System.Shrinking)) {enabled=false, module=ScaleMeshSize002}
   &ScaleMeshSize(Scale Factor=&VectorFromFloat(Value=&Lerp_Float(
      A=&Waveform(
       Waveform Count=::NewEnumerator1
       "Waveform [1]"=::NewEnumerator0
       "[1] Frequency"=2
       "[1] Amplitude Min/Max"=<0.9 1>
       "[2] Average"=false
       "[2] Input Position"=0
       "[2] Phase"=$Emitter.Age
       "[2] Phase Scale"=1
       "[2] Compound Frequency"=2
       "[2] Frequency"=4.5
       "[2] Amplitude Min/Max"=<0.93 1>
       "[2] Scale"=1
       "[2] Bias"=0
       "[2] Exponent"=1
       "[2] Pulse Duration"=0.1
       "[2] Clip Triangle Min/Max"=<0 0>
       "[2] Random Seed"=0
       "[3] Phase"=$Emitter.Age
      )
      B=&Lerp_Float(
       A=0.8
       B=1
       Alpha=&Waveform(
        "Waveform [1]"=::NewEnumerator0
        "[1] Frequency"=&Lerp_Float(A=7, B=11, Alpha=$System.FN_Timer.Value)
        "[1] Amplitude Min/Max"=<0 1>
        "[2] Phase"=$Emitter.Age
        "[3] Phase"=$Emitter.Age
       )
      )
      Alpha=$System.FN_Timer.Value
     ))) {enabled=false}
   &ScaleMeshSize(Scale Factor=&VectorFromFloat(Value=&Multiply_Float(
      A=$System.GrowingWOffset
      B=&Multiply_Float(
       A=$System.Shrinking
       B=&Lerp_Float(
        A=&Waveform(
         Waveform Count=::NewEnumerator1
         "Waveform [1]"=::NewEnumerator0
         "[1] Frequency"=2
         "[1] Amplitude Min/Max"=<0.9 1>
         "[2] Average"=false
         "[2] Input Position"=0
         "[2] Phase"=$Emitter.Age
         "[2] Phase Scale"=1
         "[2] Compound Frequency"=2
         "[2] Frequency"=4.5
         "[2] Amplitude Min/Max"=<0.93 1>
         "[2] Scale"=1
         "[2] Bias"=0
         "[2] Exponent"=1
         "[2] Pulse Duration"=0.1
         "[2] Clip Triangle Min/Max"=<0 0>
         "[2] Random Seed"=0
         "[3] Phase"=$Emitter.Age
        )
        B=&Lerp_Float(
         A=0.8
         B=1
         Alpha=&Waveform(
          "Waveform [1]"=::NewEnumerator0
          "[1] Frequency"=&Lerp_Float(A=7, B=11, Alpha=$System.FN_Timer.Value)
          "[1] Amplitude Min/Max"=<0 1>
          "[2] Phase"=$Emitter.Age
          "[3] Phase"=$Emitter.Age
         )
        )
        Alpha=$System.FN_Timer.Value
       )
      )
     ))) {module=ScaleMeshSize003}
  ]
  rendererClasses [&NiagaraMeshRendererProperties]
  renderers [
   {rendererClass=&NiagaraMeshRendererProperties, rendererIndex=0}
   {rendererClass=&NiagaraMeshRendererProperties, rendererIndex=1}
  ]
 }
 Tendrils {
  emitterUpdateScript [
   &EmitterState()
   &SpawnBurst_Instantaneous(Spawn Count=100, Use Spawn Probability=false)
   /SetVariables(Emitter.EmitterReader=DataInterface() {propertyValues {EmitterBinding {bindingMode=Other, emitterName=TendrilsEndPoints}}}) {isSetParameters=true}
  ]
  particleSpawnScript [
   &InitializeRibbon(
    Lifetime=&RandomRangeFloat(Minimum=0.3, Maximum=0.7)
    Module.Write Color=true
    Module.Write Lifetime=true
    Module.Write Mass=true
    Module.Write Material Random=true
    Module.Write Position=false
    Module.Write RibbonTwist=false
    Module.Write RibbonWidth=true
    Ribbon Width=35
   )
   &FN_SetSpawnGroup(Inclusivity=true, Normalize Spawn Group=true, Number of Spawn Groups=20, Particles Per Group=5, Set Ribbon ID and Ribbon Link Order=true)
   &InitializeParticle(Sprite Size Mode=::NewEnumerator3, Uniform Sprite Size=610.5488, UsePositionOffset=false, Write Material Random=true) {enabled=false}
  ]
  particleUpdateScript [
   &ParticleState(
    Kill Particles When Lifetime Has Elapsed=false
    Let Infinitely Lived Particles Die When Emitter Deactivates=true
    Loop Particles Lifetime=true
   )
   &self:OuterSphereLocation(New Particle Attribute Reader:NiagaraDataInterfaceParticleRead=$Emitter.EmitterReader) {module=OuterSphereLocation}
   /Game/Effects/Niagara/Modules/Curves/FN_HermiteCubicSpline_v1/FN_HermiteCubicSpline(
    Compute Direction=true
    Position Blend:NiagaraFloat=$Particles.Normalized Spawn Group
    Position End:Vector3f=/Niagara/DynamicInputs/Transforms/TransformPosition(Position=$Particles.OuterSpherePosition, Source Space=::Local, Destination Space=::Simulation)
    Position Start:Vector3f=/Niagara/DynamicInputs/Helpers/SimulationPosition()
   ) {module=FN_HermiteCubicSpline001}
   /SetVariables(Particles.Position=&Add_Vector(
     A=$Output.FN_HermiteCubicSpline001.Position
     B=/Niagara/DynamicInputs/Vectors/RotateVector(
      Vector To Rotate:Vector3f=&MakeVector(
       X=0
       Y=&Waveform(
        Global Input Position=&Add_Float(A=$Particles.Normalized Spawn Group, B=&RandomRangeFloat(Minimum=0, Maximum=1))
        Global Amplitude Scale=100
        "Waveform [1]"=::NewEnumerator0
        "[1] Frequency"=&RandomRangeFloat(Minimum=0.5, Maximum=1.5)
        "[1] Amplitude Min/Max"=<0 1>
        "[1] Scale"=&RandomRangeFloat(Minimum=-1, Maximum=1)
        "[2] Phase"=$Emitter.Age
        "[3] Phase"=$Emitter.Age
       )
       Z=&Waveform(
        Global Input Position=&Add_Float(A=$Particles.Normalized Spawn Group, B=&RandomRangeFloat(Minimum=0, Maximum=1))
        Global Amplitude Scale=100
        "Waveform [1]"=::NewEnumerator1
        "[1] Frequency"=&RandomRangeFloat(Minimum=0.5, Maximum=1.5)
        "[1] Amplitude Min/Max"=<0 1>
        "[1] Scale"=&RandomRangeFloat(Minimum=-1, Maximum=1)
        "[2] Phase"=$Emitter.Age
        "[3] Phase"=$Emitter.Age
       )
      )
      Yaw:NiagaraFloat=&MakeFloatFromVector(Vector=$Output.FN_HermiteCubicSpline001.Direction, Channel=::NewEnumerator0)
      Pitch:NiagaraFloat=&MakeFloatFromVector(Vector=$Output.FN_HermiteCubicSpline001.Direction, Channel=::NewEnumerator1)
      Roll:NiagaraFloat=&MakeFloatFromVector(Vector=$Output.FN_HermiteCubicSpline001.Direction, Channel=::NewEnumerator2)
      Rotation Rate:NiagaraFloat=$Engine.DeltaTime
     )
    )) {isSetParameters=true}
   /SetVariables(Particles.Position=&Lerp_Vector(
     A=$Output.FN_HermiteCubicSpline001.Position
     B=Linked() {linkedVariable {name=Particles.Position, type=NiagaraPosition}}
     Alpha=&FloatFromCurve(
      FloatCurve=DataInterface() {propertyValues {Curve {keys [
          {arriveTangent=3.9469, interpMode=RCIM_Cubic, leaveTangent=3.9469, tangentMode=RCTM_User, time=0, value=0}
          {interpMode=RCIM_Cubic, tangentMode=RCTM_User, time=0.5, value=1}
          {arriveTangent=-4.2192, interpMode=RCIM_Cubic, leaveTangent=-4.2192, tangentMode=RCTM_User, time=1, value=0}
         ]}}}
      CurveIndex=$Particles.Normalized Spawn Group
     )
    )) {isSetParameters=true}
   &ScaleRibbonWidth(Ribbon Width Scale=&FloatFromCurve(
     FloatCurve=DataInterface() {propertyValues {Curve {keys [{arriveTangent=-1.6835, time=0.25, value=1}, <at 0.3 0.65>, <at 0.95 0.7>, {arriveTangent=0.9446, leaveTangent=0.9446, time=1, value=1.5}]}}}
     CurveIndex=$Particles.Normalized Spawn Group
     Scale Curve=&Multiply_Float(A=$User.FX_Sphere_Large_Size, B=0.002)
    )) {enabled=false}
   &ScaleRibbonWidth(Ribbon Width Scale=$System.Shrinking) {enabled=false, module=ScaleRibbonWidth003}
   &ScaleRibbonWidth(Ribbon Width Scale=&FloatFromCurve(
     FloatCurve=DataInterface() {propertyValues {Curve {keys [{interpMode=RCIM_Cubic, tangentMode=RCTM_User, time=0.25, value=0}, {interpMode=RCIM_Cubic, tangentMode=RCTM_User, time=1.2, value=1}]}}}
     CurveIndex=$Emitter.Age
    )) {enabled=false, module=ScaleRibbonWidth002}
   &ScaleRibbonWidth(Ribbon Width Scale=&FloatFromCurve(
     FloatCurve=DataInterface() {propertyValues {Curve {keys [
         {arriveTangent=8.8939, interpMode=RCIM_Cubic, leaveTangent=8.8939, tangentMode=RCTM_User, time=0.2, value=0}
         {interpMode=RCIM_Cubic, tangentMode=RCTM_User, time=0.4, value=1}
         {arriveTangent=-1.8459, interpMode=RCIM_Cubic, leaveTangent=-1.8459, tangentMode=RCTM_User, time=0.8, value=0}
        ]}}}
     CurveIndex=&self:SampleNormalizedAge(Emitter:NiagaraDataInterfaceParticleRead=$Emitter.EmitterReader, SampleID:NiagaraInt32=$Particles.SampleID)
    )) {enabled=false, module=ScaleRibbonWidth001}
   &FN_RibbonMaterial(
    Color=&ColorFromCurve(
     ColorCurve=DataInterface() {propertyValues {
       AlphaCurve {keys [<at 0.0106 1 RCIM_Cubic>,<at 0.9796 1>]}
       BlueCurve {keys [
         {arriveTangent=-3.673, interpMode=RCIM_Cubic, time=0.2257, value=5}
         <at 0.2939 1.5>
         <at 0.4639 1>
         <at 0.5897 1>
         <at 0.898 1>
         <at 0.946 3>
         <at 0.9978 3.9124>
        ]}
       GreenCurve {keys [
         {arriveTangent=0.1965, interpMode=RCIM_Cubic, time=0.2257, value=1.9596}
         <at 0.2939 0.2737>
         <at 0.4639 0.1825>
         <at 0.5897 0>
         <at 0.898 0>
         <at 0.946 0>
         <at 0.9978 0.2517>
        ]}
       RedCurve {keys [
         {arriveTangent=0.1301, interpMode=RCIM_Cubic, time=0.2257, value=0.7117}
         <at 0.2939 0>
         <at 0.4639 0>
         <at 0.5897 0.5532>
         <at 0.898 0.5532>
         <at 0.946 1.6595>
         <at 0.9978 5>
        ]}
      }}
     CurveIndex=$Particles.Normalized Spawn Group
    )
    Opacity Pan U Speed:NiagaraFloat=&SelectFloatFromArray(Direct Array Index=$Particles.SpawnGroup, Float Selection Array=DataInterface() {propertyValues {FloatData [0.3,0.2,0.4,0.15,0.3,0]}})
    Scale Alpha:NiagaraFloat=&FloatFromCurve(
     FloatCurve=DataInterface() {propertyValues {Curve {keys [<at 0 0>,<at 0.25 0>,<at 0.26 1>,<at 1 1>]}}}
     CurveIndex=$Particles.Normalized Spawn Group
     Scale Curve=&FloatFromCurve(
      FloatCurve=DataInterface() {propertyValues {Curve {keys [
          {interpMode=RCIM_Cubic, tangentMode=RCTM_User, time=0.2, value=0}
          {interpMode=RCIM_Cubic, tangentMode=RCTM_User, time=0.5, value=1}
          {tangentMode=RCTM_User, time=0.8, value=0}
         ]}}}
      CurveIndex=&ReturnNormalizedExecIndex()
     )
    )
    Scale Color:Vector3f=&VectorFromFloat(Value=&FloatFromCurve(
      FloatCurve=DataInterface() {propertyValues {Curve {keys [
          {arriveTangent=13.1313, interpMode=RCIM_Cubic, leaveTangent=13.1313, tangentMode=RCTM_User, time=0.2, value=0}
          {interpMode=RCIM_Cubic, tangentMode=RCTM_User, time=0.3, value=1}
          {arriveTangent=-2.6487, interpMode=RCIM_Cubic, leaveTangent=-2.6487, tangentMode=RCTM_User, time=0.8, value=0}
         ]}}}
      CurveIndex=&self:SampleNormalizedAge(Emitter:NiagaraDataInterfaceParticleRead=$Emitter.EmitterReader, SampleID:NiagaraInt32=$Particles.SampleID)
     ))
   ) {enabled=false}
   &FN_RibbonMaterial(
    Color=&ColorFromCurve(
     ColorCurve=DataInterface() {propertyValues {
       AlphaCurve {keys [<at 0.0106 1 RCIM_Cubic>,<at 0.9796 1>]}
       BlueCurve {keys [
         {arriveTangent=-3.673, interpMode=RCIM_Cubic, time=0.2982, value=5}
         <at 0.3622 1.5>
         <at 0.4639 1>
         <at 0.5897 1>
         <at 0.898 1>
         <at 0.946 3>
         <at 0.9978 3.9124>
        ]}
       GreenCurve {keys [
         {arriveTangent=0.1965, interpMode=RCIM_Cubic, time=0.2982, value=1.9596}
         <at 0.3622 0.2737>
         <at 0.4639 0.1825>
         <at 0.5897 0>
         <at 0.898 0>
         <at 0.946 0>
         <at 0.9978 0.2517>
        ]}
       RedCurve {keys [
         {arriveTangent=0.1301, interpMode=RCIM_Cubic, time=0.2982, value=0.7117}
         <at 0.3622 0>
         <at 0.4639 0>
         <at 0.5897 0.5532>
         <at 0.898 0.5532>
         <at 0.946 1.6595>
         <at 0.9978 5>
        ]}
      }}
     CurveIndex=$Particles.Normalized Spawn Group
    )
    Opacity Pan U Speed:NiagaraFloat=&SelectFloatFromArray(Direct Array Index=$Particles.SpawnGroup, Float Selection Array=DataInterface() {propertyValues {FloatData [0.3,0.2,0.4,0.15,0.3,0]}})
    Scale Alpha:NiagaraFloat=&FloatFromCurve(
     FloatCurve=DataInterface() {propertyValues {Curve {keys [
         {arriveTangent=-0.0061, interpMode=RCIM_Cubic, leaveTangent=-0.0061, tangentMode=RCTM_User, time=0.2, value=0}
         {interpMode=RCIM_Cubic, tangentMode=RCTM_User, time=0.3, value=1}
         {arriveTangent=0.7242, interpMode=RCIM_Cubic, leaveTangent=0.7242, tangentMode=RCTM_User, time=0.8, value=0}
        ]}}}
     CurveIndex=&self:SampleNormalizedAge(Emitter:NiagaraDataInterfaceParticleRead=$Emitter.EmitterReader, SampleID:NiagaraInt32=$Particles.SampleID)
    )
    Scale Color:Vector3f=&VectorFromFloat(Value=&FloatFromCurve(
      FloatCurve=DataInterface() {propertyValues {Curve {keys [
          {arriveTangent=-0.2994, interpMode=RCIM_Cubic, leaveTangent=-0.2994, tangentMode=RCTM_User, time=0.2, value=0}
          {interpMode=RCIM_Cubic, tangentMode=RCTM_User, time=0.3, value=1}
          {arriveTangent=-0.1169, interpMode=RCIM_Cubic, leaveTangent=-0.1169, tangentMode=RCTM_User, time=0.8, value=0}
         ]}}}
      CurveIndex=&self:SampleNormalizedAge(Emitter:NiagaraDataInterfaceParticleRead=$Emitter.EmitterReader, SampleID:NiagaraInt32=$Particles.SampleID)
     ))
   ) {module=FN_RibbonMaterial001}
   &ScaleRibbonWidth(Ribbon Width Scale=&Multiply_Float(
     A=&Multiply_Float(
      A=&FloatFromCurve(
       FloatCurve=DataInterface() {propertyValues {Curve {keys [{arriveTangent=-1.6835, time=0.25, value=1}, <at 0.3 0.65>, <at 0.95 0.7>, {arriveTangent=0.9446, leaveTangent=0.9446, time=1, value=1.5}]}}}
       CurveIndex=$Particles.Normalized Spawn Group
       Scale Curve=&Multiply_Float(A=$User.FX_Sphere_Large_Size, B=0.002)
      )
      B=$System.Shrinking
     )
     B=&FloatFromCurve(
      FloatCurve=DataInterface() {propertyValues {Curve {keys [{interpMode=RCIM_Cubic, tangentMode=RCTM_User, time=0.25, value=0}, {interpMode=RCIM_Cubic, tangentMode=RCTM_User, time=1.5, value=1}]}}}
      CurveIndex=$Emitter.Age
     )
    )) {module=ScaleRibbonWidth004}
  ]
  rendererClasses [&NiagaraRibbonRendererProperties]
  renderers [{rendererClass=&NiagaraRibbonRendererProperties}]
 }
 TendrilsEndPoints {
  emitterUpdateScript [
   &EmitterState()
   &SpawnBurst_Instantaneous(Spawn Count=5, Use Spawn Probability=false)
   &SpawnRate(SpawnRate=4, Use Spawn Probability=false)
  ]
  particleSpawnScript [
   &InitializeParticle(
    Color=<rgba 8.73 0 10 1>
    Lifetime=&RandomRangeFloat(Minimum=0.5, Maximum=1)
    Mesh Scale Mode=::NewEnumerator3
    Mesh Uniform Scale=4
    Sprite Size Mode=::NewEnumerator3
    Uniform Sprite Size=&RandomRangeFloat(Minimum=80, Maximum=120)
    UsePositionOffset=false
    Write Material Random=true
   )
   /SetVariables(Particles.RandomAxis:Vector3f=/Niagara/DynamicInputs/Random/RandomVector(
     Vector Scale:NiagaraFloat=1
     Constraint Vector:Vector3f=<0 0 1>
     Randomness Mode:ENiagaraRandomnessMode=::NewEnumerator0
     Random Seed:NiagaraInt32=0
     Fixed Random Seed:NiagaraBool=false
    )) {isSetParameters=true}
   /SetVariables(Particles.RandomVector:Vector3f=&RandomVector(
     Vector Scale:NiagaraFloat=1
     Constraint Vector:Vector3f=<0 0 1>
     Randomness Mode:ENiagaraRandomnessMode=::NewEnumerator0
     Random Seed:NiagaraInt32=0
     Fixed Random Seed:NiagaraBool=false
    )) {isSetParameters=true}
   &InitialMeshOrientation(Orientation Vector=<0 0 1>, Use Orientation Vector=true, Use Rotation Vector=false)
  ]
  particleUpdateScript [
   &ParticleState(
    Kill Particles When Lifetime Has Elapsed=false
    Let Infinitely Lived Particles Die When Emitter Deactivates=true
    Loop Particles Lifetime=true
   )
   &KillParticles(Kill Particles=&SetBoolByFloatComparison(A=&MakeFloatFromInt(Int=&ReturnExecIndex()), B=19))
   &SphereLocation(
    MaskSpawnGroup=false
    Override Seed=false
    Sphere Radius=&Multiply_Float(A=$System.Shrinking, B=$User.FX_Sphere_Large_Size)
    Surface Only=true
    Surface Only Band Thickness=0.01
    UseNonUniformScale=false
    UseOffset=false
    UseSphereOrientation=false
   )
   /SetVariables(Particles.RandomAxis:Vector3f=&Lerp_Vector(A=$Particles.Initial.RandomAxis, B=$Particles.Initial.RandomVector, Alpha=$Particles.NormalizedAge)) {isSetParameters=true}
   /SetVariables(Particles.SpriteFacing=&TransformVector(
     Vector=&Subtract_Vector(A=Linked() {linkedVariable {name=Particles.Position, type=NiagaraPosition}}, B=&SimulationPosition())
     Source Space=::Simulation
     Destination Space=::World
    )) {isSetParameters=true}
   /Niagara/Modules/Update/Orientation/V2/OrientMeshToVector(
    Look At Direction:Vector3f=&Subtract_Vector(A=Linked() {linkedVariable {name=Particles.Position, type=NiagaraPosition}}, B=&SimulationPosition())
    Override Quat=false
    UseOptionalSideAxis=false
    UseOptionalUpAxis=false
   )
   /Niagara/Modules/Update/Velocity/VortexVelocity(Module.Use Influence Falloff=false, Velocity Amount=600, Vortex Axis:Vector3f=$Particles.RandomAxis)
   &SolveForcesAndVelocity(Clamp Velocity=false, Limit Acceleration=false)
   &FN_Radial_Material(Base Sprite Tile V=2, Depth Fade=100, Radial Mask Offset X=0.85, Radial Mask Power=10, Sprite Extrusion Speed=0.6)
   /SetVariables(Particles.Renderer2_MaterialParms:Vector4f=<1 1 1 1>) {isSetParameters=true}
   /SetVariables(Particles.Renderer3_MaterialParms:Vector4f=&MakeVector4(X=0, Y=HlslExpression() {hlslExpression=Emitter.Age*.2}, Z=2, W=2)) {isSetParameters=true}
   /SetVariables(Particles.Renderer3_MaterialParms_02:Vector4f=<2 1 0.7 1>) {isSetParameters=true}
   /SetVariables(Particles.Renderer3_MaterialParms_03:Vector4f=&MakeVector4(X=0, Y=HlslExpression() {hlslExpression=Emitter.Age*.2}, Z=3, W=3)) {isSetParameters=true}
   &ScaleColor(
    Scale Alpha=&FloatFromCurve(
     FloatCurve=DataInterface() {propertyValues {Curve {keys [
         {arriveTangent=0.7674, interpMode=RCIM_Cubic, leaveTangent=0.7674, tangentMode=RCTM_User, time=0.25, value=0}
         {arriveTangent=1.4072, interpMode=RCIM_Cubic, leaveTangent=1.4072, tangentMode=RCTM_User, time=0.4, value=0.95}
         {arriveTangent=0.5587, interpMode=RCIM_Cubic, leaveTangent=0.5587, tangentMode=RCTM_User, time=0.75, value=0}
        ]}}}
     Scale Curve=0.7
    )
    ScaleA=true
    ScaleRGB=true
   )
   /SetVariables(Particles.Renderer2_Color:LinearColor=&ColorFromCurve(ColorCurve=DataInterface() {propertyValues {
       AlphaCurve {keys [
         {arriveTangent=-0.3211, interpMode=RCIM_Cubic, leaveTangent=-0.3211, tangentMode=RCTM_User, time=0.25, value=0}
         {arriveTangent=-0.8332, interpMode=RCIM_Cubic, leaveTangent=-0.8332, tangentMode=RCTM_User, time=0.4, value=0.4}
         {arriveTangent=0.6648, interpMode=RCIM_Cubic, leaveTangent=0.6648, tangentMode=RCTM_User, time=0.75, value=0}
        ]}
       BlueCurve {keys [<at 0 1 RCIM_Cubic>]}
       GreenCurve {keys [<at 0 0 RCIM_Cubic>]}
       RedCurve {keys [<at 0 0.5532 RCIM_Cubic>]}
      }})) {isSetParameters=true}
   &ScaleSpriteSize(Scale Factor=&Vector2DFromFloat(Value=&FloatFromCurve(
      FloatCurve=DataInterface() {propertyValues {Curve {keys [
          {arriveTangent=8.8939, interpMode=RCIM_Cubic, leaveTangent=8.8939, tangentMode=RCTM_User, time=0.2, value=0}
          {arriveTangent=3.1632, interpMode=RCIM_Cubic, leaveTangent=3.1632, tangentMode=RCTM_User, time=0.3, value=0.9}
          {arriveTangent=-0.8393, interpMode=RCIM_Cubic, leaveTangent=-0.8393, tangentMode=RCTM_User, time=0.8, value=0}
         ]}}}
      Scale Curve=&Multiply_Float(A=$User.FX_Sphere_Large_Size, B=&Multiply_Float(A=0.01, B=$System.Shrinking))
     )))
   /SetVariables(Particles.Renderer2_Size:Vector2f=&Vector2DFromFloat(Value=&FloatFromCurve(
      FloatCurve=DataInterface() {propertyValues {Curve {keys [
          {arriveTangent=8.8939, interpMode=RCIM_Cubic, leaveTangent=8.8939, tangentMode=RCTM_User, time=0.2, value=0}
          {arriveTangent=3.1632, interpMode=RCIM_Cubic, leaveTangent=3.1632, tangentMode=RCTM_User, time=0.3, value=0.9}
          {arriveTangent=-0.8393, interpMode=RCIM_Cubic, leaveTangent=-0.8393, tangentMode=RCTM_User, time=0.8, value=0}
         ]}}}
      Scale Curve=35
     ))) {isSetParameters=true}
   /SetVariables(Particles.Renderer3_Size:Vector3f=&VectorFromFloat(Value=&Multiply_Float(A=&Multiply_Float(A=0.0157, B=$System.Shrinking), B=$User.FX_Sphere_Large_Size))) {isSetParameters=true}
  ]
  rendererClasses [&NiagaraSpriteRendererProperties, &NiagaraMeshRendererProperties]
  renderers [
   {rendererClass=&NiagaraSpriteRendererProperties}
   {rendererClass=&NiagaraSpriteRendererProperties, rendererIndex=1}
   {rendererClass=&NiagaraMeshRendererProperties, rendererIndex=2}
  ]
 }
 Trail {
  emitterUpdateScript [
   &EmitterState()
   &SpawnRate(SpawnRate=30, Use Spawn Probability=true)
   &SpawnBurst_Instantaneous(Spawn Count=12, Use Spawn Probability=false) {enabled=false}
  ]
  particleSpawnScript [
   &InitializeRibbon(Lifetime=1.25, Ribbon Width=1, UsePositionOffset=false, Write Material Random=true)
   &FN_SetSpawnGroup(Inclusivity=true, Normalize Spawn Group=true, Number of Spawn Groups=1, Particles Per Group=12, Set Ribbon ID and Ribbon Link Order=true)
   &AddVelocity(Velocity=<-250 0 0>) {enabled=false}
  ]
  particleUpdateScript [
   &SolveForcesAndVelocity(Clamp Velocity=false, Limit Acceleration=false)
   &ParticleState()
   /Game/Effects/Niagara/Modules/Curves/FN_HermiteCubicSpline_v2/FN_HermiteCubicSpline(
    Position Blend:NiagaraFloat=&ReturnNormalizedExecIndex()
    Position End:Vector3f=$User.Muzzle_Location
    Position Start:Vector3f=&SimulationPosition()
    Write Particles Position=true
   ) {enabled=false}
   &ScaleRibbonWidth(Ribbon Width Scale=&FloatFromCurve(
     FloatCurve=DataInterface() {propertyValues {Curve {keys [{interpMode=RCIM_Cubic, tangentMode=RCTM_User, time=0, value=1}, {interpMode=RCIM_Cubic, tangentMode=RCTM_User, time=1, value=0.2}]}}}
     CurveIndex=/Game/Effects/Niagara/DynamicInputs/Math/FN_NormalizedParticleCount()
     Scale Curve=$User.FX_Sphere_Small_Size
    ))
   &FN_RibbonMaterial(
    Color=&ColorFromCurve(
     ColorCurve=DataInterface() {propertyValues {
       AlphaCurve {keys [<at 0 1 RCIM_Cubic>,<at 1 1 RCIM_Cubic>]}
       BlueCurve {keys [<at 0 2 RCIM_Cubic>,<at 1 1 RCIM_Cubic>]}
       GreenCurve {keys [<at 0 1.1226 RCIM_Cubic>,<at 1 0 RCIM_Cubic>]}
       RedCurve {keys [<at 0 0 RCIM_Cubic>,<at 1 0.7988 RCIM_Cubic>]}
      }}
     CurveIndex=&FN_NormalizedParticleCount()
    )
    Opacity Pan U Speed=1
    Opacity Scale U=0.5
    Scale Alpha:NiagaraFloat=&FloatFromCurve(
     FloatCurve=DataInterface() {propertyValues {Curve {keys [
         {interpMode=RCIM_Cubic, tangentMode=RCTM_Break, time=0, value=0}
         {interpMode=RCIM_Cubic, tangentMode=RCTM_Break, time=0.2, value=1}
         {interpMode=RCIM_Cubic, tangentMode=RCTM_Break, time=1, value=0.1}
        ]}}}
     CurveIndex=&FN_NormalizedParticleCount()
     Scale Curve=&Lerp_Float(A=1, B=0, Alpha=$System.FN_Timer.Value)
    )
    Scale Color:Vector3f=&VectorFromFloat(Value=1.025)
   )
   /SetVariables(Particles.SecondRibbonWidth:NiagaraFloat=&FloatFromCurve(
     FloatCurve=DataInterface() {propertyValues {Curve {keys [{interpMode=RCIM_Cubic, tangentMode=RCTM_Break, time=0, value=1}, {interpMode=RCIM_Cubic, tangentMode=RCTM_Break, time=1, value=0.2}]}}}
     CurveIndex=&FN_NormalizedParticleCount()
     Scale Curve=$User.FX_Sphere_Small_Size
    )) {isSetParameters=true}
   /SetVariables(Particles.SecondLinearColor:LinearColor=&MultiplyLinearColor(
     Linear Color=&ColorFromCurve(
      ColorCurve=DataInterface() {propertyValues {
        AlphaCurve {keys [<at 0 1 RCIM_Cubic>,<at 1 1 RCIM_Cubic>]}
        BlueCurve {keys [<at 0 5 RCIM_Cubic>,<at 1 1 RCIM_Cubic>]}
        GreenCurve {keys [<at 0 1.0898 RCIM_Cubic>,<at 1 0 RCIM_Cubic>]}
        RedCurve {keys [<at 0 0 RCIM_Cubic>,<at 1 0.7988 RCIM_Cubic>]}
       }}
      CurveIndex=&FN_NormalizedParticleCount()
     )
     Scale Factor=&MakeFloat4FromVectorAndFloat(
      "Vector (RGB)"=&VectorFromFloat(Value=1.025)
      "Float (Alpha)"=&FloatFromCurve(
       FloatCurve=DataInterface() {propertyValues {Curve {keys [
           {interpMode=RCIM_Cubic, tangentMode=RCTM_User, time=0, value=0}
           {interpMode=RCIM_Cubic, tangentMode=RCTM_User, time=0.5, value=1}
           {tangentMode=RCTM_User, time=1, value=0}
          ]}}}
       CurveIndex=&FN_NormalizedParticleCount()
       Scale Curve=&Lerp_Float(A=1, B=0, Alpha=$System.FN_Timer001.Value)
      )
     )
    )) {isSetParameters=true}
  ]
  rendererClasses [&NiagaraRibbonRendererProperties]
  renderers [
   {rendererClass=&NiagaraRibbonRendererProperties}
   {rendererClass=&NiagaraRibbonRendererProperties, rendererIndex=1}
  ]
 }
 TrailPlasma {
  emitterUpdateScript [
   &EmitterState()
   &SpawnRate(SpawnRate=30, Use Spawn Probability=true)
   &SpawnBurst_Instantaneous(Spawn Count=12, Use Spawn Probability=false) {enabled=false}
  ]
  particleSpawnScript [
   &InitializeRibbon(Lifetime=0.75, Ribbon Width=1, UsePositionOffset=false, Write Material Random=true)
   &FN_SetSpawnGroup(Inclusivity=true, Normalize Spawn Group=true, Number of Spawn Groups=1, Particles Per Group=12, Set Ribbon ID and Ribbon Link Order=true)
   &AddVelocity(Velocity=<-250 0 0>) {enabled=false}
  ]
  particleUpdateScript [
   &SolveForcesAndVelocity(Clamp Velocity=false, Limit Acceleration=false)
   &ParticleState()
   &FN_HermiteCubicSpline(
    Position Blend:NiagaraFloat=&ReturnNormalizedExecIndex()
    Position End:Vector3f=$User.Muzzle_Location
    Position Start:Vector3f=&SimulationPosition()
    Write Particles Position=true
   ) {enabled=false}
   &ScaleRibbonWidth(Ribbon Width Scale=&FloatFromCurve(
     FloatCurve=DataInterface() {propertyValues {Curve {keys [{interpMode=RCIM_Cubic, tangentMode=RCTM_Break, time=0, value=1}, {interpMode=RCIM_Cubic, tangentMode=RCTM_Break, time=1, value=0.2}]}}}
     CurveIndex=&FN_NormalizedParticleCount()
     Scale Curve=$User.FX_Sphere_Small_Size
    ))
   &FN_RibbonMaterial(
    Color=&ColorFromCurve(
     ColorCurve=DataInterface() {propertyValues {
       AlphaCurve {keys [<at 0 1 RCIM_Cubic>,<at 1 1 RCIM_Cubic>]}
       BlueCurve {keys [<at 0 2 RCIM_Cubic>,<at 1 1 RCIM_Cubic>]}
       GreenCurve {keys [<at 0 1.1226 RCIM_Cubic>,<at 1 0 RCIM_Cubic>]}
       RedCurve {keys [<at 0 0 RCIM_Cubic>,<at 1 0.7988 RCIM_Cubic>]}
      }}
     CurveIndex=&FN_NormalizedParticleCount()
    )
    Opacity Pan U Speed=1
    Opacity Scale U=0.5
    Scale Alpha:NiagaraFloat=&FloatFromCurve(
     FloatCurve=DataInterface() {propertyValues {Curve {keys [
         {interpMode=RCIM_Cubic, tangentMode=RCTM_Break, time=0, value=0}
         {interpMode=RCIM_Cubic, tangentMode=RCTM_Break, time=0.2, value=1}
         {interpMode=RCIM_Cubic, tangentMode=RCTM_Break, time=1, value=0.1}
        ]}}}
     CurveIndex=&FN_NormalizedParticleCount()
     Scale Curve=&Lerp_Float(A=1, B=0, Alpha=$System.FN_Timer.Value)
    )
    Scale Color:Vector3f=&VectorFromFloat(Value=1.025)
   )
   /SetVariables(Particles.RibbonTwist:NiagaraFloat=&FloatFromCurve(FloatCurve=DataInterface() {propertyValues {Curve {keys [<at 0 1>,<at 1 0>]}}}, Scale Curve=5)) {isSetParameters=true}
  ]
  rendererClasses [&NiagaraRibbonRendererProperties]
  renderers [{rendererClass=&NiagaraRibbonRendererProperties}]
 }
}
summary {
 systemName=NS_BadNews_Projectile
 emitters [
  {emitterName=Radial_Plasma_Energy_01, rendererClasses [&NiagaraSpriteRendererProperties]}
  {emitterName=Radial_Plasma_Energy_02, rendererClasses [&NiagaraSpriteRendererProperties]}
  {
   emitterName=Plasma_Center_Energy
   rendererClasses [&NiagaraSpriteRendererProperties, &NiagaraLightRendererProperties]
  }
  {emitterName=Floating_Plasma_Particles, rendererClasses [&NiagaraSpriteRendererProperties]}
  {emitterName=Small_Sphere, rendererClasses [&NiagaraMeshRendererProperties]}
  {emitterName=Trail, rendererClasses [&NiagaraRibbonRendererProperties]}
  {
   bEnabled=false
   emitterName=Light
   rendererClasses [&NiagaraComponentRendererProperties, &NiagaraLightRendererProperties]
  }
  {bEnabled=false, emitterName=Radial_Plasma_Energy_01001, rendererClasses [&NiagaraSpriteRendererProperties]}
  {emitterName=TrailPlasma, rendererClasses [&NiagaraRibbonRendererProperties]}
  {emitterName=Tendrils, rendererClasses [&NiagaraRibbonRendererProperties]}
  {
   emitterName=TendrilsEndPoints
   rendererClasses [&NiagaraSpriteRendererProperties, &NiagaraMeshRendererProperties]
  }
  {bEnabled=false, emitterName=Debug, rendererClasses [&NiagaraMeshRendererProperties]}
  {bEnabled=false, emitterName=Large_Sphere, rendererClasses [&NiagaraMeshRendererProperties]}
 ]
 userVariables [
  User.BaseNexusLocation {type=Vector3f, description="", defaultValue {struct=Vector3f, value=<0 0 0>}}
  User.bIsFrontEndPreview {type=NiagaraBool}
  User.ChargeUpLength {type=NiagaraFloat, defaultValue=0, description=""}
  User.Destabilization {type=NiagaraFloat, defaultValue=0, description=""}
  User.Expired {type=NiagaraBool}
  User.ExpiredShrink {type=NiagaraFloat, defaultValue=1, description=""}
  User.ExpireStart {type=NiagaraBool}
  User.FEVec {type=Vector3f, description="", defaultValue {struct=Vector3f, value=<0 0 1000>}}
  User.FX_Sphere_Large_Size {type=NiagaraFloat, defaultValue=250, description=""}
  User.FX_Sphere_Small_Size {type=NiagaraFloat, defaultValue=50, description=""}
  User.HitLocation {type=Vector3f, description="", defaultValue {struct=Vector3f, value=<0 0 0>}}
  User.Hitting {type=NiagaraBool}
  User.LightVariant {type=LinearColor, description="", defaultValue {struct=LinearColor, value=<rgba 1 1 1 1>}}
  User.Muzzle_Location {type=Vector3f, description="", defaultValue {struct=Vector3f, value=<500 0 0>}}
  User.PlayerSpeed {type=NiagaraFloat, defaultValue=0, description=""}
  User.PositionVelocity {type=Vector3f, description="", defaultValue {struct=Vector3f, value=<0 0 0>}}
  User.RibbonIDOffset {type=NiagaraInt32}
  User.Skeletal Mesh {
   type=NiagaraDataInterfaceSkeletalMesh
   defaultValue {
    struct=NiagaraExt_VariableValue_DataInterface
    value {dataInterface=&self:NiagaraDataInterfaceSkeletalMesh_1}
   }
  }
  User.SpawnRate {type=NiagaraFloat, defaultValue=360, description=""}
  User.StartLocation {type=Vector3f, description="", defaultValue {struct=Vector3f, value=<0 0 0>}}
 ]
}
