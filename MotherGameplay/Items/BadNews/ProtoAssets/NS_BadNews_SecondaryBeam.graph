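# Niagara prototype effect for the BadNews secondary beam, drawing short-lived ribbon plasma arcs between User.StartLocation and User.HitLocation.
# Two burst-spawned emitters, a wide plasma ribbon and a longer-lived thin one, place points along a Hermite spline by normalized spawn group.
# Randomized waveform offsets rotated using the spline direction jitter the points, with a curve lerp pinning both ends.
# Ribbon materials add a color gradient along the beam, panning opacity and an alpha fade.
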NiagaraSystem /MotherGameplay/Items/BadNews/ProtoAssets/NS_BadNews_SecondaryBeam
dependencies {
 usedDataInterfaces [&NiagaraDataInterfaceCurve, &NiagaraDataInterfaceColorCurve]
 usedDynamicInputs [
  &Multiply_Int
  &MakeIntFromFloat
  &RandomRangeInt
 ]
 usedModules [
  &self:SystemSpawnScript.SystemScriptSource.SystemScriptGraph.NiagaraNodeAssignment_0.SetVariables_#1
  &SystemState
  &self:SystemSpawnScript.SystemScriptSource.SystemScriptGraph.NiagaraNodeAssignment_1.SetVariables_#2
 ]
}
emitters {
 Blue_Outer_Plasma001 {
  emitterUpdateScript [
   &EmitterState()
   &SpawnBurst_Instantaneous(Spawn Count=10, Spawn Time=0.1, Use Spawn Probability=false)
  ]
  particleSpawnScript [
   &FN_SetSpawnGroup(Inclusivity=true, Normalize Spawn Group=true, Particles Per Group=10, Pattern=::NewEnumerator0, Set Ribbon ID and Ribbon Link Order=true)
   &InitializeRibbon(
    Lifetime=0.2
    Module.Write Color=true
    Module.Write Lifetime=true
    Module.Write Mass=true
    Module.Write Material Random=true
    Module.Write Position=false
    Module.Write RibbonTwist=false
    Module.Write RibbonWidth=true
    Ribbon Width=80
   )
  ]
  particleUpdateScript [
   &ParticleState()
   /Game/Effects/Niagara/Modules/Curves/FN_HermiteCubicSpline_v2/FN_HermiteCubicSpline(
    Position Blend:NiagaraFloat=$Particles.Normalized Spawn Group
    Position End:Vector3f=$User.HitLocation
    Position Start:Vector3f=$User.StartLocation
    Write Particles Position=true
   ) {enabled=false}
   /Game/Effects/Niagara/Modules/Curves/FN_HermiteCubicSpline_v1/FN_HermiteCubicSpline(
    Compute Direction=true
    Position Blend:NiagaraFloat=$Particles.Normalized Spawn Group
    Position End:Vector3f=$User.HitLocation
    Position Start:Vector3f=$User.StartLocation
   ) {module=FN_HermiteCubicSpline001}
   /SetVariables(Particles.Position=&Add_Vector(
     A=$Output.FN_HermiteCubicSpline001.Position
     B=/Niagara/DynamicInputs/Vectors/RotateVector(
      Vector To Rotate:Vector3f=&MakeVector(
       X=0
       Y=&Waveform(
        Global Input Position=&Add_Float(A=$Particles.Normalized Spawn Group, B=&RandomRangeFloat(Minimum=0, Maximum=1))
        Global Amplitude Scale=50
        "Waveform [1]"=::NewEnumerator0
        "[1] Frequency"=1
        "[1] Amplitude Min/Max"=<0 1>
        "[1] Scale"=&RandomRangeFloat(Minimum=-1, Maximum=1)
        "[2] Phase"=$Emitter.Age
        "[3] Phase"=$Emitter.Age
       )
       Z=&Waveform(
        Global Input Position=&Add_Float(A=$Particles.Normalized Spawn Group, B=&RandomRangeFloat(Minimum=0, Maximum=1))
        Global Amplitude Scale=50
        "Waveform [1]"=::NewEnumerator1
        "[1] Frequency"=1
        "[1] Amplitude Min/Max"=<0 1>
        "[1] Scale"=&RandomRangeFloat(Minimum=-1, Maximum=1)
        "[2] Phase"=$Emitter.Age
        "[3] Phase"=$Emitter.Age
       )
      )
      Yaw:NiagaraFloat=&MakeFloatFromVector(Vector=$Output.FN_HermiteCubicSpline001.Direction, Channel=::NewEnumerator0)
      Pitch:NiagaraFloat=&MakeFloatFromVector(Vector=$Output.FN_HermiteCubicSpline001.Direction, Channel=::NewEnumerator1)
      Roll:NiagaraFloat=&MakeFloatFromVector(Vector=$Output.FN_HermiteCubicSpline001.Direction, Channel=::NewEnumerator2)
      Rotation Rate:NiagaraFloat=$Engine.DeltaTime
     )
    )) {isSetParameters=true, module=SetVariables_#3}
   /SetVariables(Particles.Position=&Lerp_Vector(
     A=$Output.FN_HermiteCubicSpline001.Position
     B=Linked() {linkedVariable {name=Particles.Position, type=NiagaraPosition}}
     Alpha=&FloatFromCurve(
      FloatCurve=DataInterface() {propertyValues {Curve {keys [
          {arriveTangent=6.9713, interpMode=RCIM_Cubic, leaveTangent=6.9713, tangentMode=RCTM_User, time=0, value=0}
          {interpMode=RCIM_Cubic, tangentMode=RCTM_User, time=0.5, value=1}
          {arriveTangent=-6.3689, interpMode=RCIM_Cubic, leaveTangent=-6.3689, tangentMode=RCTM_User, time=1, value=0}
         ]}}}
      CurveIndex=$Particles.Normalized Spawn Group
     )
    )) {isSetParameters=true, module=SetVariables_#4}
   &FN_RibbonMaterial(
    Color=&ColorFromCurve(
     ColorCurve=DataInterface() {propertyValues {
       AlphaCurve {keys [<at 0.0106 1 RCIM_Cubic>,<at 0.9796 1>]}
       BlueCurve {keys [
         {arriveTangent=-3.673, interpMode=RCIM_Cubic, time=0.2257, value=7}
         <at 0.2939 4>
         <at 0.4639 4>
         <at 0.5897 4>
         <at 0.898 4>
         <at 0.946 5>
         <at 0.9978 7.8247>
        ]}
       GreenCurve {keys [
         {arriveTangent=0.1965, interpMode=RCIM_Cubic, time=0.2257, value=2.7435}
         <at 0.2939 0.7299>
         <at 0.4639 0.7299>
         <at 0.5897 0>
         <at 0.898 0>
         <at 0.946 0>
         <at 0.9978 0.5035>
        ]}
       RedCurve {keys [
         {arriveTangent=0.1301, interpMode=RCIM_Cubic, time=0.2257, value=0.9963}
         <at 0.2939 0>
         <at 0.4639 0>
         <at 0.5897 2.2127>
         <at 0.898 2.2127>
         <at 0.946 2.7658>
         <at 0.9978 10>
        ]}
      }}
     CurveIndex=$Particles.Normalized Spawn Group
    )
    Opacity Pan U Offset=0.9143
    Opacity Pan U Speed=5
    Opacity Scale U=0.7
    Scale Alpha:NiagaraFloat=&FloatFromCurve(FloatCurve=DataInterface() {propertyValues {Curve {keys [
         {arriveTangent=13.8107, interpMode=RCIM_Cubic, leaveTangent=13.8107, tangentMode=RCTM_User, time=0, value=0}
         <at 0.2 1>
         {arriveTangent=-2.6535, interpMode=RCIM_Cubic, leaveTangent=-2.6535, tangentMode=RCTM_User, time=1, value=0}
        ]}}})
   )
   &ScaleRibbonWidth(Ribbon Width Scale=&FloatFromCurve(FloatCurve=DataInterface() {propertyValues {Curve {keys [
         {interpMode=RCIM_Cubic, tangentMode=RCTM_User, time=0, value=1}
         {arriveTangent=-2.3513, interpMode=RCIM_Cubic, leaveTangent=-2.3513, tangentMode=RCTM_User, time=1, value=0}
        ]}}})) {enabled=false}
   /Game/Effects/Niagara/Modules/Curves/FN_HermiteCubicSpline_v2/FN_HermiteCubicSpline() {module=FN_HermiteCubicSpline002}
  ]
  rendererClasses [&NiagaraRibbonRendererProperties]
  renderers [{rendererClass=&NiagaraRibbonRendererProperties}]
 }
 Thin {
  emitterUpdateScript [
   &EmitterState()
   &SpawnBurst_Instantaneous(Spawn Count=10, Use Spawn Probability=false)
  ]
  particleSpawnScript [
   &FN_SetSpawnGroup(Inclusivity=true, Normalize Spawn Group=true, Particles Per Group=10, Pattern=::NewEnumerator0, Set Ribbon ID and Ribbon Link Order=true)
   &InitializeRibbon(
    Lifetime=0.35
    Module.Write Color=true
    Module.Write Lifetime=true
    Module.Write Mass=true
    Module.Write Material Random=true
    Module.Write Position=false
    Module.Write RibbonTwist=false
    Module.Write RibbonWidth=true
    Ribbon Width=30
   )
  ]
  particleUpdateScript [
   &ParticleState()
   &FN_HermiteCubicSpline(
    Position Blend:NiagaraFloat=$Particles.Normalized Spawn Group
    Position End:Vector3f=$User.HitLocation
    Position Start:Vector3f=$User.StartLocation
    Write Particles Position=true
   ) {enabled=false}
   /Game/Effects/Niagara/Modules/Curves/FN_HermiteCubicSpline_v1/FN_HermiteCubicSpline(
    Compute Direction=true
    Position Blend:NiagaraFloat=$Particles.Normalized Spawn Group
    Position End:Vector3f=$User.HitLocation
    Position Start:Vector3f=$User.StartLocation
   ) {module=FN_HermiteCubicSpline001}
   /SetVariables(Particles.Position=&Add_Vector(
     A=$Output.FN_HermiteCubicSpline001.Position
     B=&RotateVector(
      Vector To Rotate:Vector3f=&MakeVector(
       X=0
       Y=&Waveform(
        Global Input Position=&Add_Float(A=$Particles.Normalized Spawn Group, B=&RandomRangeFloat(Minimum=0, Maximum=1))
        Global Amplitude Scale=65
        "Waveform [1]"=::NewEnumerator0
        "[1] Frequency"=1
        "[1] Amplitude Min/Max"=<0 1>
        "[1] Scale"=&RandomRangeFloat(Minimum=-1, Maximum=1)
        "[2] Phase"=$Emitter.Age
        "[3] Phase"=$Emitter.Age
       )
       Z=&Waveform(
        Global Input Position=&Add_Float(A=$Particles.Normalized Spawn Group, B=&RandomRangeFloat(Minimum=0, Maximum=1))
        Global Amplitude Scale=65
        "Waveform [1]"=::NewEnumerator1
        "[1] Frequency"=1
        "[1] Amplitude Min/Max"=<0 1>
        "[1] Scale"=&RandomRangeFloat(Minimum=-1, Maximum=1)
        "[2] Phase"=$Emitter.Age
        "[3] Phase"=$Emitter.Age
       )
      )
      Yaw:NiagaraFloat=&MakeFloatFromVector(Vector=$Output.FN_HermiteCubicSpline001.Direction, Channel=::NewEnumerator0)
      Pitch:NiagaraFloat=&MakeFloatFromVector(Vector=$Output.FN_HermiteCubicSpline001.Direction, Channel=::NewEnumerator1)
      Roll:NiagaraFloat=&MakeFloatFromVector(Vector=$Output.FN_HermiteCubicSpline001.Direction, Channel=::NewEnumerator2)
      Rotation Rate:NiagaraFloat=$Engine.DeltaTime
     )
    )) {isSetParameters=true, module=SetVariables_#3}
   /SetVariables(Particles.Position=&Lerp_Vector(
     A=$Output.FN_HermiteCubicSpline001.Position
     B=Linked() {linkedVariable {name=Particles.Position, type=NiagaraPosition}}
     Alpha=&FloatFromCurve(
      FloatCurve=DataInterface() {propertyValues {Curve {keys [
          {arriveTangent=6.9713, interpMode=RCIM_Cubic, leaveTangent=6.9713, tangentMode=RCTM_User, time=0, value=0}
          {interpMode=RCIM_Cubic, tangentMode=RCTM_User, time=0.5, value=1}
          {arriveTangent=-6.3689, interpMode=RCIM_Cubic, leaveTangent=-6.3689, tangentMode=RCTM_User, time=1, value=0}
         ]}}}
      CurveIndex=$Particles.Normalized Spawn Group
     )
    )) {isSetParameters=true, module=SetVariables_#4}
   &FN_RibbonMaterial(
    Color=&ColorFromCurve(
     ColorCurve=DataInterface() {propertyValues {
       AlphaCurve {keys [<at 0.0106 1 RCIM_Cubic>,<at 0.9796 1>]}
       BlueCurve {keys [
         {arriveTangent=-3.673, interpMode=RCIM_Cubic, time=0.2257, value=7}
         <at 0.2939 4>
         <at 0.4639 4>
         <at 0.5897 4>
         <at 0.898 4>
         <at 0.946 5>
         <at 0.9978 7.8247>
        ]}
       GreenCurve {keys [
         {arriveTangent=0.1965, interpMode=RCIM_Cubic, time=0.2257, value=2.7435}
         <at 0.2939 0.7299>
         <at 0.4639 0.7299>
         <at 0.5897 0>
         <at 0.898 0>
         <at 0.946 0>
         <at 0.9978 0.5035>
        ]}
       RedCurve {keys [
         {arriveTangent=0.1301, interpMode=RCIM_Cubic, time=0.2257, value=0.9963}
         <at 0.2939 0>
         <at 0.4639 0>
         <at 0.5897 2.2127>
         <at 0.898 2.2127>
         <at 0.946 2.7658>
         <at 0.9978 10>
        ]}
      }}
     CurveIndex=$Particles.Normalized Spawn Group
    )
    Opacity Pan U Offset=0.9143
    Opacity Pan U Speed=5
    Opacity Scale U=0.7
    Scale Alpha:NiagaraFloat=&FloatFromCurve(FloatCurve=DataInterface() {propertyValues {Curve {keys [
         {arriveTangent=13.8107, interpMode=RCIM_Cubic, leaveTangent=13.8107, tangentMode=RCTM_User, time=0, value=0}
         <at 0.2 1>
         {arriveTangent=-2.6535, interpMode=RCIM_Cubic, leaveTangent=-2.6535, tangentMode=RCTM_User, time=1, value=0}
        ]}}})
   )
   &ScaleRibbonWidth(Ribbon Width Scale=&FloatFromCurve(FloatCurve=DataInterface() {propertyValues {Curve {keys [
         {interpMode=RCIM_Cubic, tangentMode=RCTM_User, time=0, value=1}
         {arriveTangent=-2.3513, interpMode=RCIM_Cubic, leaveTangent=-2.3513, tangentMode=RCTM_User, time=1, value=0}
        ]}}})) {enabled=false}
   /Game/Effects/Niagara/Modules/Curves/FN_HermiteCubicSpline_v2/FN_HermiteCubicSpline() {module=FN_HermiteCubicSpline002}
  ]
  rendererClasses [&NiagaraRibbonRendererProperties]
  renderers [{rendererClass=&NiagaraRibbonRendererProperties}]
 }
}
summary {
 systemName=NS_BadNews_SecondaryBeam
 emitters [
  {emitterName=Blue_Outer_Plasma001, rendererClasses [&NiagaraRibbonRendererProperties]}
  {emitterName=Thin, rendererClasses [&NiagaraRibbonRendererProperties]}
 ]
 userVariables [
  User.bIsFrontEnd {type=NiagaraBool}
  User.bIsFrontEndPreview {type=NiagaraBool}
  User.DamageLevel {type=NiagaraInt32, defaultValue=3}
  User.HitLocation {type=Vector3f, description="", defaultValue {struct=Vector3f, value=<-50 50 200>}}
  User.StartLocation {type=Vector3f, description="", defaultValue {struct=Vector3f, value=<0 0 0>}}
 ]
}
